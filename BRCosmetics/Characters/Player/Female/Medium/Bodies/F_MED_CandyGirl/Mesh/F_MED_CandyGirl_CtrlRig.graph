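# Control Rig for the F_MED_CandyGirl cosmetic body that adds procedural corrective deformation and keeps the skirt following the legs.
# A shared DeformRig function runs first, then weighted parent constraints blend forearm button bones between arm, twist and hand bones.
# Six spherical pose readers per thigh then feed smoothed alphas into ModifyTransforms offsets on skirt master bones.
# Spur-related variables exist but the graph never uses them.
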ControlRigBlueprint /BRCosmetics/Characters/Player/Female/Medium/Bodies/F_MED_CandyGirl/Mesh/F_MED_CandyGirl_CtrlRig
graphs [
 RigVMFunctionLibrary {links [], nodes []}
 RigVMModel {nodes [
   RigUnit_BeginExecution = &RigUnit_BeginExecution()
   RigVMFunction_Sequence = &RigVMFunction_Sequence(ExecuteContext=@RigUnit_BeginExecution.ExecutePin)
   FN_DeformRig_FNC = RigVMFunctionReferenceNode(
    DeformRig_ Curve="(deformRig_l_upperarm_up_alpha_72_#1=False,deformRig_r_upperarm_up_alpha_73_#2=False,deformRig_l_upperarm_fr_alpha_74_#3=False,deformRig_r_upperarm_fr_alpha_75_CB51B8 ...(truncated)"
    DeformRig_Bool="(upperarm_1_#4=True,elbow_3_#5=True,wrist_7_#6=True,groin_8_#7=True,glute_17_#8=True,knee_9_40C408324CA261 ...(truncated)"
    DeformRig_Debug=(upperarm_debug_18_#4=False,wrist_debug_20_#6=False,lat_debug_26_#9=False,groinGlute_debug_27_#7=False)
    DeformRig_Value="(upperarm_up_18_#10=(Rotation=(X=0,Y=0,Z=0,W=1),Translation=(X=0,Y=0,Z=8),Scale3D=(X=0.7,Y=1,Z=1)),upperarm_fwd_fr_19_#11=(Rotation=(X=0,Y=0,Z=0,W=1),Translation=(X=0,Y=10,Z=0),Scale3D=(X=0.5,Y=1,Z= ...(truncated)"
    ExecuteContext=@RigVMFunction_Sequence.A
   )
   > &RigUnit_ParentConstraint(
    AdvancedSettings=(RotationOrderForFilter=XZY)
    Child=(Type=Bone,Name="deform_button_l_01")
    Parents=((Item=(Type=Bone,Name="upperarm_l"),Weight=0.0081),(Item=(Type=Bone,Name="lowerarm_l"),Weight=0.1933),(Item=(Type=Bone,Name="lowerarm_twist_02_l"),Weight=0.2526),(Item=(Type=Bone,Name="lowerarm_twist_01_l"),Weight=0.546))
    Weight=1
   )
   > &RigUnit_ParentConstraint(
    AdvancedSettings=(RotationOrderForFilter=XZY)
    Child=(Type=Bone,Name="deform_button_l_02")
    Parents="((Item=(Type=Bone,Name=\"deform_wrist_up_l\"),Weight=0.0194),(Item=(Type=Bone,Name=\"hand_l\"),Weight=0.0214),(Item=(Type=Bone,Name=\"lowerarm_twist_02_l\"),Weight=0.1035),(Item=(Type=Bone,Name=\"lowerarm_l\"),Weight=0.1241),(Item=(Type=Bone,Name=\"lowerarm_twist_0 ...(truncated)"
    Weight=1
   )
   > &RigUnit_ParentConstraint(
    AdvancedSettings=(RotationOrderForFilter=XZY)
    Child=(Type=Bone,Name="deform_button_l_03")
    Parents="((Item=(Type=Bone,Name=\"ring_metacarpal_l\"),Weight=0.0059),(Item=(Type=Bone,Name=\"lowerarm_twist_02_l\"),Weight=0.0189),(Item=(Type=Bone,Name=\"lowerarm_l\"),Weight=0.0277),(Item=(Type=Bone,Name=\"deform_wrist_up_l\"),Weight=0.1241),(Item=(Type=Bone,Name=\"hand_ ...(truncated)"
    Weight=1
   )
   > &RigUnit_ParentConstraint(
    AdvancedSettings=(RotationOrderForFilter=XZY)
    Child=(Type=Bone,Name="deform_button_r_01")
    Parents=((Item=(Type=Bone,Name="upperarm_r"),Weight=0.0081),(Item=(Type=Bone,Name="lowerarm_r"),Weight=0.1933),(Item=(Type=Bone,Name="lowerarm_twist_02_r"),Weight=0.2526),(Item=(Type=Bone,Name="lowerarm_twist_01_r"),Weight=0.546))
    Weight=1
   )
   > &RigUnit_ParentConstraint(
    AdvancedSettings=(RotationOrderForFilter=XZY)
    Child=(Type=Bone,Name="deform_button_r_02")
    Parents="((Item=(Type=Bone,Name=\"deform_wrist_up_r\"),Weight=0.0194),(Item=(Type=Bone,Name=\"hand_r\"),Weight=0.0214),(Item=(Type=Bone,Name=\"lowerarm_twist_02_r\"),Weight=0.1035),(Item=(Type=Bone,Name=\"lowerarm_r\"),Weight=0.1241),(Item=(Type=Bone,Name=\"lowerarm_twist_0 ...(truncated)"
    Weight=1
   )
   > &RigUnit_ParentConstraint(
    AdvancedSettings=(RotationOrderForFilter=XZY)
    Child=(Type=Bone,Name="deform_button_r_03")
    Parents="((Item=(Type=Bone,Name=\"ring_metacarpal_r\"),Weight=0.0059),(Item=(Type=Bone,Name=\"lowerarm_twist_02_r\"),Weight=0.0189),(Item=(Type=Bone,Name=\"lowerarm_r\"),Weight=0.0277),(Item=(Type=Bone,Name=\"deform_wrist_up_r\"),Weight=0.1241),(Item=(Type=Bone,Name=\"hand_ ...(truncated)"
    Weight=1
   )
   n1 = &RigVMFunction_Sequence(ExecuteContext=@RigVMFunction_Sequence.B)
   n2 = &RigUnit_SphericalPoseReader(
    ActiveRegionScaleFactors=(PositiveWidth=3,NegativeWidth=0.5,PositiveHeight=1.2,NegativeHeight=1)
    ActiveRegionSize=0.1
    Debug=(bDrawDebug=false,bDraw2D=False,bDrawLocalAxes=False,DebugScale=25,DebugSegments=20,DebugThickness=0.25)
    DriverAxis=(X=-1,Y=0,Z=0)
    DriverItem=(Type=Bone,Name="thigh_l")
    FalloffRegionScaleFactors=(PositiveWidth=1,NegativeWidth=1,PositiveHeight=0.75,NegativeHeight=0.75)
    FalloffSize=0.4
    FlipHeightScaling=False
    FlipWidthScaling=False
    OptionalParentItem=(Type=Bone,Name="None")
    RotationOffset=(X=90,Y=0,Z=-15)
    ExecutePin=@n1.A
   )
   RigVMFunction_AlphaInterp = &RigVMFunction_AlphaInterp(InterpSpeedDecreasing=30, InterpSpeedIncreasing=30, Scale=1, bClampResult=False, bInterpResult=True, Value=@n2.OutputParam)
   > &RigUnit_ModifyTransforms(
    ItemToModify="((Item=(Type=Bone,Name=\"dyn_skirt_front_MSTR\"),Transform=(Rotation=(X=0.2346,Y=0.4152,Z=0.2251,W=-0.8496),Translation=(X=0,Y=0,Z=0),Scale3D=(X=1,Y=1,Z=1))),(Item=(Type=Bone,Name=\"dyn_skirt_sidefr_l_MSTR\"),Transform=(Rotation=(X=0.0676,Y=-0.5155,Z=0.0609,W= ...(truncated)"
    WeightMinimum=0
    Weight=@RigVMFunction_AlphaInterp.Result
   )
   > n3 = &RigUnit_SphericalPoseReader(
    ActiveRegionScaleFactors=(PositiveWidth=3,NegativeWidth=0.5,PositiveHeight=1.3,NegativeHeight=1.2)
    ActiveRegionSize=0.1
    Debug=(bDrawDebug=false,bDraw2D=False,bDrawLocalAxes=False,DebugScale=25,DebugSegments=20,DebugThickness=0.25)
    DriverAxis=(X=-1,Y=0,Z=0)
    DriverItem=(Type=Bone,Name="thigh_l")
    FalloffRegionScaleFactors=(PositiveWidth=1,NegativeWidth=1,PositiveHeight=0.5,NegativeHeight=0.5)
    FalloffSize=0.57
    FlipHeightScaling=False
    FlipWidthScaling=False
    OptionalParentItem=(Type=Bone,Name="None")
    RotationOffset=(X=130,Y=5,Z=-12)
   )
   n4 = &RigVMFunction_AlphaInterp(InterpSpeedDecreasing=30, InterpSpeedIncreasing=30, Scale=1, bClampResult=False, bInterpResult=True, Value=@n3.OutputParam)
   > &RigUnit_ModifyTransforms(
    ItemToModify="((Item=(Type=Bone,Name=\"dyn_skirt_front_MSTR\"),Transform=(Rotation=(X=0,Y=-0.0087,Z=0,W=1),Translation=(X=0,Y=0,Z=0),Scale3D=(X=1,Y=1,Z=1))),(Item=(Type=Bone,Name=\"dyn_skirt_sidefr_l_MSTR\"),Transform=(Rotation=(X=-0.0759,Y=-0.2622,Z=0.0279,W=0.9616),Transl ...(truncated)"
    WeightMinimum=0
    Weight=@n4.Result
   )
   > n5 = &RigUnit_SphericalPoseReader(
    ActiveRegionScaleFactors=(PositiveWidth=3,NegativeWidth=1,PositiveHeight=1,NegativeHeight=1)
    ActiveRegionSize=0.1
    Debug=(bDrawDebug=true,bDraw2D=False,bDrawLocalAxes=False,DebugScale=25,DebugSegments=20,DebugThickness=0.25)
    DriverAxis=(X=-1,Y=0,Z=0)
    DriverItem=(Type=Bone,Name="thigh_l")
    FalloffRegionScaleFactors=(PositiveWidth=1,NegativeWidth=1,PositiveHeight=1,NegativeHeight=1)
    FalloffSize=0.5
    FlipHeightScaling=False
    FlipWidthScaling=False
    OptionalParentItem=(Type=Bone,Name="None")
    RotationOffset=(X=180,Y=0,Z=0)
   )
   n6 = &RigVMFunction_AlphaInterp(InterpSpeedDecreasing=30, InterpSpeedIncreasing=30, Scale=1, bClampResult=False, bInterpResult=True, Value=@n5.OutputParam)
   > &RigUnit_ModifyTransforms(
    ItemToModify="((Item=(Type=Bone,Name=\"dyn_skirt_front_MSTR\"),Transform=(Rotation=(X=0,Y=0,Z=0,W=1),Translation=(X=0,Y=0,Z=0),Scale3D=(X=1,Y=1,Z=1))),(Item=(Type=Bone,Name=\"dyn_skirt_sidefr_l_MSTR\"),Transform=(Rotation=(X=-0.1231,Y=-0.0952,Z=-0.1793,W=0.9714),Translation ...(truncated)"
    WeightMinimum=0
    Weight=@n6.Result
   )
   > n7 = &RigUnit_SphericalPoseReader(
    ActiveRegionScaleFactors=(PositiveWidth=3,NegativeWidth=1,PositiveHeight=1.5,NegativeHeight=1.5)
    ActiveRegionSize=0.1
    Debug=(bDrawDebug=false,bDraw2D=False,bDrawLocalAxes=False,DebugScale=25,DebugSegments=20,DebugThickness=0.25)
    DriverAxis=(X=-1,Y=0,Z=0)
    DriverItem=(Type=Bone,Name="thigh_l")
    FalloffRegionScaleFactors=(PositiveWidth=1,NegativeWidth=1,PositiveHeight=1,NegativeHeight=1)
    FalloffSize=0.5
    FlipHeightScaling=False
    FlipWidthScaling=False
    OptionalParentItem=(Type=Bone,Name="None")
    RotationOffset=(X=-125,Y=0,Z=0)
   )
   n8 = &RigVMFunction_AlphaInterp(InterpSpeedDecreasing=30, InterpSpeedIncreasing=30, Scale=1, bClampResult=False, bInterpResult=True, Value=@n7.OutputParam)
   > &RigUnit_ModifyTransforms(
    ItemToModify="((Item=(Type=Bone,Name=\"dyn_skirt_sidebk_l_MSTR\"),Transform=(Rotation=(X=0,Y=-0.2701,Z=0,W=0.9628),Translation=(X=0,Y=0,Z=0),Scale3D=(X=1,Y=1,Z=1))),(Item=(Type=Bone,Name=\"dyn_skirt_side_l_MSTR\"),Transform=(Rotation=(X=0.0038,Y=-0.078,Z=-0.0491,W=0.9957),T ...(truncated)"
    WeightMinimum=0
    Weight=@n8.Result
   )
   > n9 = &RigUnit_SphericalPoseReader(
    ActiveRegionScaleFactors=(PositiveWidth=3,NegativeWidth=1,PositiveHeight=3,NegativeHeight=3)
    ActiveRegionSize=0.1
    Debug=(bDrawDebug=false,bDraw2D=False,bDrawLocalAxes=False,DebugScale=25,DebugSegments=20,DebugThickness=0.25)
    DriverAxis=(X=-1,Y=0,Z=0)
    DriverItem=(Type=Bone,Name="thigh_l")
    FalloffRegionScaleFactors=(PositiveWidth=1,NegativeWidth=0.95,PositiveHeight=1,NegativeHeight=1)
    FalloffSize=0.5
    FlipHeightScaling=False
    FlipWidthScaling=False
    OptionalParentItem=(Type=Bone,Name="None")
    RotationOffset=(X=-90,Y=0,Z=0)
   )
   n10 = &RigVMFunction_AlphaInterp(InterpSpeedDecreasing=30, InterpSpeedIncreasing=30, Scale=1, bClampResult=False, bInterpResult=True, Value=@n9.OutputParam)
   > &RigUnit_ModifyTransforms(
    ItemToModify="((Item=(Type=Bone,Name=\"dyn_skirt_sidebk_l_MSTR\"),Transform=(Rotation=(X=-0.2072,Y=-0.11,Z=-0.0234,W=0.9718),Translation=(X=0,Y=0,Z=0),Scale3D=(X=1,Y=1,Z=1))),(Item=(Type=Bone,Name=\"dyn_skirt_side_l_MSTR\"),Transform=(Rotation=(X=0,Y=0,Z=0,W=1),Translation= ...(truncated)"
    WeightMinimum=0
    Weight=@n10.Result
   )
   > n11 = &RigUnit_SphericalPoseReader(
    ActiveRegionScaleFactors=(PositiveWidth=3,NegativeWidth=1,PositiveHeight=1.2,NegativeHeight=1)
    ActiveRegionSize=0.1
    Debug=(bDrawDebug=false,bDraw2D=False,bDrawLocalAxes=False,DebugScale=25,DebugSegments=20,DebugThickness=0.25)
    DriverAxis=(X=-1,Y=0,Z=0)
    DriverItem=(Type=Bone,Name="thigh_l")
    FalloffRegionScaleFactors=(PositiveWidth=1,NegativeWidth=1,PositiveHeight=0.3,NegativeHeight=0.3)
    FalloffSize=0.45
    FlipHeightScaling=False
    FlipWidthScaling=False
    OptionalParentItem=(Type=Bone,Name="None")
    RotationOffset=(X=45,Y=0,Z=-30)
   )
   n12 = &RigVMFunction_AlphaInterp(InterpSpeedDecreasing=30, InterpSpeedIncreasing=30, Scale=1, bClampResult=False, bInterpResult=True, Value=@n11.OutputParam)
   > &RigUnit_ModifyTransforms(
    ItemToModify="((Item=(Type=Bone,Name=\"dyn_skirt_front_MSTR\"),Transform=(Rotation=(X=0.0324,Y=0.4716,Z=-0.0455,W=-0.88),Translation=(X=-2.3616,Y=-0.0073,Z=1.097),Scale3D=(X=1,Y=1,Z=1))),(Item=(Type=Bone,Name=\"dyn_skirt_sidefr_r_MSTR\"),Transform=(Rotation=(X=0.2742,Y=-0.4 ...(truncated)"
    WeightMinimum=0
    Weight=@n12.Result
   )
   n13 = &RigUnit_SphericalPoseReader(
    ActiveRegionScaleFactors=(PositiveWidth=3,NegativeWidth=0.5,PositiveHeight=1.2,NegativeHeight=1)
    ActiveRegionSize=0.1
    Debug=(bDrawDebug=false,bDraw2D=False,bDrawLocalAxes=False,DebugScale=25,DebugSegments=20,DebugThickness=0.25)
    DriverAxis=(X=1,Y=0,Z=0)
    DriverItem=(Type=Bone,Name="thigh_r")
    FalloffRegionScaleFactors=(PositiveWidth=1,NegativeWidth=1,PositiveHeight=0.75,NegativeHeight=0.75)
    FalloffSize=0.58
    FlipHeightScaling=true
    FlipWidthScaling=true
    OptionalParentItem=(Type=Bone,Name="None")
    RotationOffset=(X=-90,Y=0,Z=-15)
    ExecutePin=@n1.B
   )
   n14 = &RigVMFunction_AlphaInterp(InterpSpeedDecreasing=30, InterpSpeedIncreasing=30, Scale=1, bClampResult=False, bInterpResult=True, Value=@n13.OutputParam)
   > &RigUnit_ModifyTransforms(
    ItemToModify="((Item=(Type=Bone,Name=\"dyn_skirt_front_MSTR\"),Transform=(Rotation=(X=0.2546,Y=-0.4256,Z=0.0557,W=0.8666),Translation=(X=-0.4456,Y=-0.1981,Z=0.1285),Scale3D=(X=1,Y=1,Z=1))),(Item=(Type=Bone,Name=\"dyn_skirt_sidefr_r_MSTR\"),Transform=(Rotation=(X=0.0656,Y=-0 ...(truncated)"
    WeightMinimum=0
    Weight=@n14.Result
   )
   > n15 = &RigUnit_SphericalPoseReader(
    ActiveRegionScaleFactors=(PositiveWidth=3,NegativeWidth=0.5,PositiveHeight=1.3,NegativeHeight=1.2)
    ActiveRegionSize=0.1
    Debug=(bDrawDebug=false,bDraw2D=False,bDrawLocalAxes=False,DebugScale=25,DebugSegments=20,DebugThickness=0.25)
    DriverAxis=(X=1,Y=0,Z=0)
    DriverItem=(Type=Bone,Name="thigh_r")
    FalloffRegionScaleFactors=(PositiveWidth=1,NegativeWidth=1,PositiveHeight=0.5,NegativeHeight=0.5)
    FalloffSize=0.57
    FlipHeightScaling=true
    FlipWidthScaling=true
    OptionalParentItem=(Type=Bone,Name="None")
    RotationOffset=(X=-50,Y=5,Z=-12)
   )
   n16 = &RigVMFunction_AlphaInterp(InterpSpeedDecreasing=30, InterpSpeedIncreasing=30, Scale=1, bClampResult=False, bInterpResult=True, Value=@n15.OutputParam)
   > &RigUnit_ModifyTransforms(
    ItemToModify="((Item=(Type=Bone,Name=\"dyn_skirt_sidefr_r_MSTR\"),Transform=(Rotation=(X=-0.0799,Y=-0.0435,Z=-0.0035,W=0.9958),Translation=(X=0,Y=0,Z=0),Scale3D=(X=1,Y=1,Z=1))),(Item=(Type=Bone,Name=\"dyn_skirt_side_r_MSTR\"),Transform=(Rotation=(X=0.1079,Y=-0.0314,Z=0.1168 ...(truncated)"
    WeightMinimum=0
    Weight=@n16.Result
   )
   > n17 = &RigUnit_SphericalPoseReader(
    ActiveRegionScaleFactors=(PositiveWidth=3,NegativeWidth=0.5,PositiveHeight=1,NegativeHeight=1)
    ActiveRegionSize=0.1
    Debug=(bDrawDebug=false,bDraw2D=False,bDrawLocalAxes=False,DebugScale=25,DebugSegments=20,DebugThickness=0.25)
    DriverAxis=(X=1,Y=0,Z=0)
    DriverItem=(Type=Bone,Name="thigh_r")
    FalloffRegionScaleFactors=(PositiveWidth=1,NegativeWidth=1,PositiveHeight=1,NegativeHeight=1)
    FalloffSize=0.5
    FlipHeightScaling=true
    FlipWidthScaling=true
    OptionalParentItem=(Type=Bone,Name="None")
    RotationOffset=(X=0,Y=0,Z=0)
   )
   n18 = &RigVMFunction_AlphaInterp(InterpSpeedDecreasing=30, InterpSpeedIncreasing=30, Scale=1, bClampResult=False, bInterpResult=True, Value=@n17.OutputParam)
   > &RigUnit_ModifyTransforms(
    ItemToModify="((Item=(Type=Bone,Name=\"dyn_skirt_sidefr_r_MSTR\"),Transform=(Rotation=(X=0.2044,Y=0.1984,Z=0.0238,W=-0.9583),Translation=(X=0,Y=0,Z=0),Scale3D=(X=1,Y=1,Z=1))),(Item=(Type=Bone,Name=\"dyn_skirt_side_r_MSTR\"),Transform=(Rotation=(X=0,Y=-0.4043,Z=0,W=0.9146),T ...(truncated)"
    WeightMinimum=0
    Weight=@n18.Result
   )
   > n19 = &RigUnit_SphericalPoseReader(
    ActiveRegionScaleFactors=(PositiveWidth=3,NegativeWidth=1,PositiveHeight=1.5,NegativeHeight=1.5)
    ActiveRegionSize=0.1
    Debug=(bDrawDebug=false,bDraw2D=False,bDrawLocalAxes=False,DebugScale=25,DebugSegments=20,DebugThickness=0.25)
    DriverAxis=(X=1,Y=0,Z=0)
    DriverItem=(Type=Bone,Name="thigh_r")
    FalloffRegionScaleFactors=(PositiveWidth=1,NegativeWidth=1,PositiveHeight=1,NegativeHeight=1)
    FalloffSize=0.5
    FlipHeightScaling=true
    FlipWidthScaling=true
    OptionalParentItem=(Type=Bone,Name="None")
    RotationOffset=(X=55,Y=0,Z=0)
   )
   n20 = &RigVMFunction_AlphaInterp(InterpSpeedDecreasing=30, InterpSpeedIncreasing=30, Scale=1, bClampResult=False, bInterpResult=True, Value=@n19.OutputParam)
   > &RigUnit_ModifyTransforms(
    ItemToModify="((Item=(Type=Bone,Name=\"dyn_skirt_sidebk_r_MSTR\"),Transform=(Rotation=(X=0.0379,Y=-0.3183,Z=-0.0584,W=0.9454),Translation=(X=0,Y=0,Z=0),Scale3D=(X=1,Y=1,Z=1))),(Item=(Type=Bone,Name=\"dyn_skirt_side_r_MSTR\"),Transform=(Rotation=(X=0.1133,Y=0.1413,Z=0.1395,W ...(truncated)"
    WeightMinimum=0
    Weight=@n20.Result
   )
   > n21 = &RigUnit_SphericalPoseReader(
    ActiveRegionScaleFactors=(PositiveWidth=3,NegativeWidth=1,PositiveHeight=3,NegativeHeight=3)
    ActiveRegionSize=0.1
    Debug=(bDrawDebug=false,bDraw2D=False,bDrawLocalAxes=False,DebugScale=25,DebugSegments=20,DebugThickness=0.25)
    DriverAxis=(X=1,Y=0,Z=0)
    DriverItem=(Type=Bone,Name="thigh_r")
    FalloffRegionScaleFactors=(PositiveWidth=1,NegativeWidth=0.95,PositiveHeight=1,NegativeHeight=1)
    FalloffSize=0.5
    FlipHeightScaling=true
    FlipWidthScaling=true
    OptionalParentItem=(Type=Bone,Name="None")
    RotationOffset=(X=90,Y=0,Z=0)
   )
   n22 = &RigVMFunction_AlphaInterp(InterpSpeedDecreasing=30, InterpSpeedIncreasing=30, Scale=1, bClampResult=False, bInterpResult=True, Value=@n21.OutputParam)
   > &RigUnit_ModifyTransforms(
    ItemToModify="((Item=(Type=Bone,Name=\"dyn_skirt_back_MSTR\"),Transform=(Rotation=(X=-0.351,Y=-0.2555,Z=-0.0852,W=0.8968),Translation=(X=0,Y=0,Z=0),Scale3D=(X=1,Y=1,Z=1))),(Item=(Type=Bone,Name=\"dyn_skirt_sidebk_r_MSTR\"),Transform=(Rotation=(X=-0.0848,Y=-0.1575,Z=-0.0678, ...(truncated)"
    WeightMinimum=0
    Weight=@n22.Result
   )
   > n23 = &RigUnit_SphericalPoseReader(
    ActiveRegionScaleFactors=(PositiveWidth=3,NegativeWidth=1,PositiveHeight=1.2,NegativeHeight=1)
    ActiveRegionSize=0.1
    Debug=(bDrawDebug=false,bDraw2D=False,bDrawLocalAxes=False,DebugScale=25,DebugSegments=20,DebugThickness=0.25)
    DriverAxis=(X=1,Y=0,Z=0)
    DriverItem=(Type=Bone,Name="thigh_r")
    FalloffRegionScaleFactors=(PositiveWidth=1,NegativeWidth=1,PositiveHeight=0.3,NegativeHeight=0.3)
    FalloffSize=0.45
    FlipHeightScaling=true
    FlipWidthScaling=true
    OptionalParentItem=(Type=Bone,Name="None")
    RotationOffset=(X=-135,Y=0,Z=-30)
   )
   n24 = &RigVMFunction_AlphaInterp(InterpSpeedDecreasing=20, InterpSpeedIncreasing=20, Scale=1, bClampResult=False, bInterpResult=True, Value=@n23.OutputParam)
   > &RigUnit_ModifyTransforms(
    ItemToModify="((Item=(Type=Bone,Name=\"dyn_skirt_front_MSTR\"),Transform=(Rotation=(X=-0.1043,Y=-0.3856,Z=-0.0225,W=0.9165),Translation=(X=-1.3389,Y=0.9277,Z=1.3265),Scale3D=(X=1,Y=1,Z=1))),(Item=(Type=Bone,Name=\"dyn_skirt_sidefr_r_MSTR\"),Transform=(Rotation=(X=0.0409,Y=0 ...(truncated)"
    WeightMinimum=0
    Weight=@n24.Result
   )
  ]}
]
hierarchy [
 root {loc=<0 0 0>, rot=<0 0 0 1>}
 pelvis {parent=root, loc=<0 -0.5301 98.0571>, rot=<0.0127 0.707 -0.0127 -0.707>}
 spine_01 {parent=pelvis, loc=<6.1208 -0.9328 0>, rot=<0 0 0.034 -0.9994>}
 spine_02 {parent=spine_01, loc=<6.5996 -0.2095 0>, rot=<0 0 0.0238 -0.9997>}
 spine_03 {parent=spine_02, loc=<6.4914 -0.1237 0>, rot=<0 0 0.0149 -0.9999>}
 spine_04 {parent=spine_03, loc=<6.9598 0.0794 0>, rot=<0 0 -0.0132 -0.9999>}
 spine_05 {parent=spine_04, loc=<6.9987 0 0>, rot=<0 0 -0.0015 -1>}
 clavicle_l {parent=spine_05, loc=<5.4831 -0.1053 -3.2946>, rot=<0.0465 -0.7289 -0.009 -0.6829>}
 upperarm_l {parent=clavicle_l, loc=<9.8658 -1.2136 -0.4105>, rot=<0.0208 -0.486 -0.0242 -0.8734>}
 lowerarm_l {parent=upperarm_l, loc=<26.465 -0.8345 0.259>, rot=<0.0036 0.0161 -0.2298 -0.9731>}
 hand_l {parent=lowerarm_l, loc=<22.7355 -0.3203 -0.6105>, rot=<-0.6784 -0.0223 0.0735 -0.7307>}
 index_metacarpal_l {parent=hand_l, loc=<1.6535 0.0183 -2.96>, rot=<0.0158 -0.0124 -0.0173 -0.9996>}
 index_01_l {parent=index_metacarpal_l, loc=<8.3008 0.0357 -0.0524>, rot=<-0.0044 0.0485 0.0924 -0.9945>}
 index_02_l {parent=index_01_l, loc=<3.3587 -0.3004 0.0452>, rot=<-0.002 0.0027 0.3225 -0.9466>}
 index_03_l {parent=index_02_l, loc=<2.5578 -0.123 -0.0252>, rot=<-0.002 0.0067 0.2558 -0.9667>}
 middle_metacarpal_l {parent=hand_l, loc=<2.1856 -0.1725 -0.5984>, rot=<0.006 0.0168 -0.0329 -0.9993>}
 middle_01_l {parent=middle_metacarpal_l, loc=<7.3275 -0.2447 0.0129>, rot=<-0.0208 0.0042 0.0858 -0.9961>}
 middle_02_l {parent=middle_01_l, loc=<4.0123 0.1401 0.1175>, rot=<-0.0218 0.0258 0.3512 -0.9357>}
 middle_03_l {parent=middle_02_l, loc=<2.6077 0.0534 0.1373>, rot=<0.0082 0.0188 0.1708 -0.9851>}
 pinky_metacarpal_l {parent=hand_l, loc=<1.7616 -0.9753 3.3814>, rot=<-0.0397 0.0814 -0.069 -0.9935>}
 pinky_01_l {parent=pinky_metacarpal_l, loc=<7.1709 -0.4831 -0.1094>, rot=<-0.0654 -0.0035 0.2242 -0.9723>}
 pinky_02_l {parent=pinky_01_l, loc=<2.8274 0.0893 -0.1043>, rot=<0.0037 0.0011 0.2854 -0.9584>}
 pinky_03_l {parent=pinky_02_l, loc=<2.2046 0.0307 -0.0078>, rot=<-0.0002 0.0098 0.1445 -0.9895>}
 ring_metacarpal_l {parent=hand_l, loc=<2.2074 -0.3625 1.3638>, rot=<-0.0284 0.0722 -0.0458 -0.9959>}
 ring_01_l {parent=ring_metacarpal_l, loc=<7.0295 -0.6197 -0.1446>, rot=<-0.0489 -0.0252 0.1892 -0.9804>}
 ring_02_l {parent=ring_01_l, loc=<3.6174 0.4647 -0.1027>, rot=<0 0 0.2806 -0.9598>}
 ring_03_l {parent=ring_02_l, loc=<2.4152 0.114 0.0199>, rot=<0 0 0.1189 -0.9929>}
 thumb_01_l {parent=hand_l, loc=<1.3118 -0.3191 -2.459>, rot=<0.6546 -0.4042 0.0664 -0.6354>}
 thumb_02_l {parent=thumb_01_l, loc=<5.9336 -0.0448 0.0401>, rot=<-0.0037 0.0213 0.298 -0.9543>}
 thumb_03_l {parent=thumb_02_l, loc=<2.2223 0.3682 -0.0907>, rot=<0.0118 -0.013 0.0066 -0.9998>}
 deform_wrist_dn_base_l {parent=hand_l, loc=<0 -1.5327 0>, rot=<0 0 0 -1>}
 deform_wrist_dn_l {parent=deform_wrist_dn_base_l, loc=<0 0 0>, rot=<0 0 0 1>}
 deform_wrist_up_base_l {parent=hand_l, loc=<0 2.3008 0>, rot=<0 0 0 -1>}
 deform_wrist_up_l {parent=deform_wrist_up_base_l, loc=<0 0 0>, rot=<0 0 0 1>}
 lowerarm_twist_01_l {parent=lowerarm_l, loc=<18.316 0 0>, rot=<0 0 0 1>}
 deform_button_l_01 {parent=lowerarm_twist_01_l, loc=<-5.4702 -0.2055 4.1305>, rot=<-0.5298 -0.5365 -0.4436 0.4845>}
 deform_button_l_02 {parent=lowerarm_twist_01_l, loc=<-1.6741 -0.6137 4.0996>, rot=<0.5238 0.5014 0.4745 -0.4991>}
 deform_button_l_03 {parent=lowerarm_twist_01_l, loc=<2.1262 -0.7227 4.0345>, rot=<0.5125 0.5118 0.4928 -0.4822>}
 lowerarm_twist_02_l {parent=lowerarm_l, loc=<10.6707 0 0>, rot=<0 0 0 1>}
 deform_elbow_in_base_l {parent=lowerarm_l, loc=<0.1835 3.6062 -0.2572>, rot=<0 0 0 -1>}
 deform_elbow_in_l {parent=deform_elbow_in_base_l, loc=<0 0 0>, rot=<0 0 0 1>}
 deform_elbow_out_base_l {parent=lowerarm_l, loc=<0 -4 0>, rot=<0 0 0 -1>}
 deform_elbow_out_l {parent=deform_elbow_out_base_l, loc=<0 0 0>, rot=<0 0 0 1>}
 upperarm_twist_01_l {parent=upperarm_l, loc=<7.2135 -0.1134 0>, rot=<0 0 0 1>}
 deform_upperarm_bk_base_l {parent=upperarm_twist_01_l, loc=<-6.5369 -4.3754 0.521>, rot=<-0.0001 0.1412 0.0005 -0.99>}
 deform_upperarm_bk_l {parent=deform_upperarm_bk_base_l, loc=<0 0 0>, rot=<0 0 0 1>}
 deform_upperarm_fr_base_l {parent=upperarm_twist_01_l, loc=<-5.777 5.0905 0.7064>, rot=<-0.0001 0.1236 0.0005 -0.9923>}
 deform_upperarm_fr_l {parent=deform_upperarm_fr_base_l, loc=<0 0 0>, rot=<0 0 0 1>}
 deform_upperarm_up_base_l {parent=upperarm_twist_01_l, loc=<-9.3973 0.4834 3.621>, rot=<0 0.0708 0.0003 -0.9975>}
 deform_upperarm_up_l {parent=deform_upperarm_up_base_l, loc=<0 0 0>, rot=<0 0 0 1>}
 upperarm_twist_02_l {parent=upperarm_l, loc=<19.7998 -0.1134 0>, rot=<0 0 0 1>}
 deform_candy_shld_l_01 {parent=clavicle_l, loc=<7.817 7.0391 -0.8957>, rot=<-0.0782 0.1977 -0.4809 -0.8506>}
 deform_candy_shld_l_02 {parent=clavicle_l, loc=<7.57 5.9017 1.2009>, rot=<-0.1053 0.2189 -0.5299 -0.8125>}
 deform_candy_shld_l_03 {parent=clavicle_l, loc=<7.497 4.2241 3.6356>, rot=<-0.1496 0.3066 -0.488 -0.8034>}
 deform_candy_shld_l_04 {parent=clavicle_l, loc=<7.8737 1.9047 5.2902>, rot=<-0.2136 0.4517 -0.4201 -0.7576>}
 deform_candy_shld_l_05 {parent=clavicle_l, loc=<8.3552 -4.855 5.3793>, rot=<0.2161 0.4186 0.4589 -0.7533>}
 deform_candy_shld_l_06 {parent=clavicle_l, loc=<8.5707 -7.3486 3.4746>, rot=<0.1302 0.2737 0.495 -0.8143>}
 deform_candy_shld_l_07 {parent=clavicle_l, loc=<8.676 -8.629 1.1198>, rot=<0.0532 0.1295 0.5756 -0.8056>}
 Charm_Attach_00 {parent=clavicle_l, type=Null, loc=<3.2197 6.6717 1.3935>, rot=<0.2385 0.233 -0.7531 0.5671>}
 clavicle_r {parent=spine_05, loc=<5.4831 -0.1053 3.2946>, rot=<0.7289 0.0465 -0.6829 0.009>}
 upperarm_r {parent=clavicle_r, loc=<-9.8658 1.2136 0.4104>, rot=<0.0208 -0.486 -0.0242 -0.8734>}
 lowerarm_r {parent=upperarm_r, loc=<-26.4651 0.8345 -0.259>, rot=<0.0036 0.0161 -0.2298 -0.9731>}
 hand_r {parent=lowerarm_r, loc=<-22.7356 0.3203 0.6106>, rot=<-0.6784 -0.0223 0.0735 -0.7307>}
 index_metacarpal_r {parent=hand_r, loc=<-1.6534 -0.0181 2.9601>, rot=<0.0158 -0.0124 -0.0173 -0.9996>}
 index_01_r {parent=index_metacarpal_r, loc=<-8.3008 -0.0357 0.0524>, rot=<-0.0044 0.0485 0.0924 -0.9945>}
 index_02_r {parent=index_01_r, loc=<-3.3589 0.3003 -0.0453>, rot=<-0.002 0.0027 0.3225 -0.9466>}
 index_03_r {parent=index_02_r, loc=<-2.5578 0.123 0.0252>, rot=<-0.002 0.0067 0.2558 -0.9667>}
 middle_metacarpal_r {parent=hand_r, loc=<-2.1856 0.1726 0.5984>, rot=<0.006 0.0168 -0.0329 -0.9993>}
 middle_01_r {parent=middle_metacarpal_r, loc=<-7.3273 0.2447 -0.0129>, rot=<-0.0208 0.0042 0.0858 -0.9961>}
 middle_02_r {parent=middle_01_r, loc=<-4.0124 -0.14 -0.1175>, rot=<-0.0218 0.0258 0.3512 -0.9357>}
 middle_03_r {parent=middle_02_r, loc=<-2.6077 -0.0533 -0.1373>, rot=<0.0082 0.0188 0.1708 -0.9851>}
 pinky_metacarpal_r {parent=hand_r, loc=<-1.7617 0.9753 -3.3814>, rot=<-0.0397 0.0814 -0.069 -0.9935>}
 pinky_01_r {parent=pinky_metacarpal_r, loc=<-7.1707 0.4832 0.1094>, rot=<-0.0654 -0.0035 0.2242 -0.9723>}
 pinky_02_r {parent=pinky_01_r, loc=<-2.8274 -0.0892 0.1043>, rot=<0.0037 0.0011 0.2854 -0.9584>}
 pinky_03_r {parent=pinky_02_r, loc=<-2.2045 -0.031 0.0078>, rot=<-0.0002 0.0098 0.1445 -0.9895>}
 ring_metacarpal_r {parent=hand_r, loc=<-2.2073 0.3627 -1.3638>, rot=<-0.0284 0.0722 -0.0458 -0.9959>}
 ring_01_r {parent=ring_metacarpal_r, loc=<-7.0297 0.6197 0.1446>, rot=<-0.0489 -0.0252 0.1892 -0.9804>}
 ring_02_r {parent=ring_01_r, loc=<-3.6171 -0.4648 0.1027>, rot=<0 0 0.2806 -0.9598>}
 ring_03_r {parent=ring_02_r, loc=<-2.4153 -0.1141 -0.0199>, rot=<0 0 0.1189 -0.9929>}
 thumb_01_r {parent=hand_r, loc=<-1.3119 0.3192 2.459>, rot=<0.6546 -0.4042 0.0664 -0.6354>}
 thumb_02_r {parent=thumb_01_r, loc=<-5.9335 0.0447 -0.0401>, rot=<-0.0037 0.0213 0.298 -0.9543>}
 thumb_03_r {parent=thumb_02_r, loc=<-2.2223 -0.3682 0.0907>, rot=<0.0118 -0.013 0.0066 -0.9998>}
 deform_wrist_dn_base_r {parent=hand_r, loc=<0 1.5329 0>, rot=<0 0 0 -1>}
 deform_wrist_dn_r {parent=deform_wrist_dn_base_r, loc=<0 0 0>, rot=<0 0 0 1>}
 deform_wrist_up_base_r {parent=hand_r, loc=<-0.0001 -2.3006 -0.0001>, rot=<0 0 0 -1>}
 deform_wrist_up_r {parent=deform_wrist_up_base_r, loc=<0 0 0>, rot=<0 0 0 1>}
 lowerarm_twist_01_r {parent=lowerarm_r, loc=<-18.3161 0 0>, rot=<0 0 0 1>}
 deform_button_r_01 {parent=lowerarm_twist_01_r, loc=<5.4707 0.2053 -4.1307>, rot=<-0.5298 -0.5365 -0.4436 0.4845>}
 deform_button_r_02 {parent=lowerarm_twist_01_r, loc=<1.6742 0.6137 -4.0997>, rot=<0.5238 0.5014 0.4745 -0.4991>}
 deform_button_r_03 {parent=lowerarm_twist_01_r, loc=<-2.1261 0.7227 -4.0345>, rot=<0.5125 0.5118 0.4928 -0.4822>}
 lowerarm_twist_02_r {parent=lowerarm_r, loc=<-10.6707 0 0>, rot=<0 0 0 1>}
 deform_elbow_in_base_r {parent=lowerarm_r, loc=<-0.1834 -3.6062 0.2572>, rot=<0 0 0 -1>}
 deform_elbow_in_r {parent=deform_elbow_in_base_r, loc=<0 0 0>, rot=<0 0 0 1>}
 deform_elbow_out_base_r {parent=lowerarm_r, loc=<0 4 0>, rot=<0 0 0 -1>}
 deform_elbow_out_r {parent=deform_elbow_out_base_r, loc=<0 0 0>, rot=<0 0 0 1>}
 upperarm_twist_01_r {parent=upperarm_r, loc=<-7.2135 0.1134 0>, rot=<0 0 0 1>}
 deform_upperarm_bk_base_r {parent=upperarm_twist_01_r, loc=<6.5368 4.3754 -0.521>, rot=<-0.0001 0.1412 0.0005 -0.99>}
 deform_upperarm_bk_r {parent=deform_upperarm_bk_base_r, loc=<0 0 0>, rot=<0 0 0 1>}
 deform_upperarm_fr_base_r {parent=upperarm_twist_01_r, loc=<5.7456 -5.0908 -0.6268>, rot=<-0.0001 0.0972 0.0004 -0.9953>}
 deform_upperarm_fr_r {parent=deform_upperarm_fr_base_r, loc=<0 0 0>, rot=<0 0 0 1>}
 deform_upperarm_up_base_r {parent=upperarm_twist_01_r, loc=<9.3972 -0.4834 -3.6209>, rot=<0 0.0708 0.0003 -0.9975>}
 deform_upperarm_up_r {parent=deform_upperarm_up_base_r, loc=<0 0 0>, rot=<0 0 0 1>}
 upperarm_twist_02_r {parent=upperarm_r, loc=<-19.7998 0.1134 0>, rot=<0 0 0 1>}
 deform_candy_shld_r_01 {parent=clavicle_r, loc=<-7.817 -7.0391 0.8957>, rot=<-0.0782 0.1977 -0.4809 -0.8506>}
 deform_candy_shld_r_02 {parent=clavicle_r, loc=<-7.5701 -5.9017 -1.2005>, rot=<-0.1053 0.2189 -0.5299 -0.8125>}
 deform_candy_shld_r_03 {parent=clavicle_r, loc=<-7.497 -4.2241 -3.6351>, rot=<-0.1496 0.3066 -0.488 -0.8034>}
 deform_candy_shld_r_04 {parent=clavicle_r, loc=<-7.8737 -1.9047 -5.2901>, rot=<-0.2136 0.4517 -0.4201 -0.7576>}
 deform_candy_shld_r_05 {parent=clavicle_r, loc=<-8.3552 4.855 -5.3796>, rot=<0.2161 0.4186 0.4589 -0.7533>}
 deform_candy_shld_r_06 {parent=clavicle_r, loc=<-8.5707 7.3486 -3.4747>, rot=<0.1302 0.2737 0.495 -0.8143>}
 deform_candy_shld_r_07 {parent=clavicle_r, loc=<-8.676 8.629 -1.1195>, rot=<0.0532 0.1295 0.5756 -0.8056>}
 neck_01 {parent=spine_05, loc=<11.3138 0.1529 0>, rot=<0 0 -0.2003 -0.9797>}
 neck_02 {parent=neck_01, loc=<4.5627 0 0>, rot=<0 0 0.0504 -0.9987>}
 head {parent=neck_02, loc=<5.3914 0 0>, rot=<0 0 0.115 -0.9934>}
 dyn_string_l_MSTR {parent=spine_05, loc=<8.5484 7.4722 -4.578>, rot=<-0.1543 -0.7373 -0.6294 -0.191>}
 dyn_string_l_01 {parent=dyn_string_l_MSTR, loc=<0 0 0>, rot=<0 0 0 1>}
 dyn_string_l_02 {parent=dyn_string_l_01, loc=<7.6791 0 0>, rot=<0.019 0.0135 -0.0127 -0.9996>}
 dyn_string_l_03 {parent=dyn_string_l_02, loc=<5.7914 0 0>, rot=<0.0409 -0.0229 -0.0267 -0.9985>}
 dyn_string_r_MSTR {parent=spine_05, loc=<8.5487 7.4722 4.578>, rot=<-0.7373 0.1543 0.191 -0.6294>}
 dyn_string_r_01 {parent=dyn_string_r_MSTR, loc=<0 0 0>, rot=<0 0 0 1>}
 dyn_string_r_02 {parent=dyn_string_r_01, loc=<-7.6793 0 0.0001>, rot=<0.019 0.0135 -0.0127 -0.9996>}
 dyn_string_r_03 {parent=dyn_string_r_02, loc=<-5.7913 0 0>, rot=<0.0409 -0.0229 -0.0267 -0.9985>}
 attach_cape {parent=spine_05, loc=<8.1632 -11.9857 -0.0006>, rot=<0.0617 -0.7044 -0.0616 -0.7044>}
 cape {parent=attach_cape, type=Null, loc=<-0.0004 0.743 1.8673>, rot=<0.0471 0.0019 -0.0401 0.9981>}
 attach_backpack {parent=spine_05, loc=<-0.0604 -11.1795 0.0005>, rot=<-0.0092 -0.7071 0.0092 -0.707>}
 Backpack_BR {parent=attach_backpack, type=Null, loc=<-0.0001 3.4068 -0.3636>, rot=<-0.0131 0 0 0.9999>}
 deform_pec_aim_base_l {parent=spine_04, loc=<3.2456 9.3309 -1.2237>, rot=<0.0603 -0.4033 0.3318 -0.8507>}
 deform_pec_aim_l {parent=deform_pec_aim_base_l, loc=<0 0 0>, rot=<0 0 0 1>}
 deform_pec_base_l {parent=deform_pec_aim_l, loc=<9.3665 0 0>, rot=<0 0 0 1>}
 deform_pec_l {parent=deform_pec_base_l, loc=<0 0 0>, rot=<0 0 0 1>}
 deform_pec_aim_base_r {parent=spine_04, loc=<3.2452 9.3309 1.2237>, rot=<-0.4033 -0.0603 0.8507 0.3318>}
 deform_pec_aim_r {parent=deform_pec_aim_base_r, loc=<0 0 0>, rot=<0 0 0 -1>}
 deform_pec_base_r {parent=deform_pec_aim_r, loc=<-9.3667 -0.0003 -0.0004>, rot=<0 0 0 1>}
 deform_pec_r {parent=deform_pec_base_r, loc=<0 0 0>, rot=<0 0 0 1>}
 deform_candy_front_01 {parent=spine_01, loc=<9.2229 10.5092 -2.3376>, rot=<-0.5447 -0.4104 -0.4194 -0.5991>}
 deform_candy_front_02 {parent=spine_01, loc=<6.4301 10.934 -2.3327>, rot=<-0.5393 -0.4278 -0.4508 -0.5683>}
 deform_candy_front_03 {parent=spine_01, loc=<9.1489 10.6102 0.0018>, rot=<-0.5726 -0.4149 -0.4149 -0.5726>}
 deform_candy_front_04 {parent=spine_01, loc=<6.3478 11.0636 0.0235>, rot=<-0.554 -0.4394 -0.4394 -0.554>}
 deform_candy_front_05 {parent=spine_01, loc=<6.4601 10.9474 2.3622>, rot=<-0.5776 -0.4582 -0.4199 -0.5293>}
 deform_candy_front_06 {parent=spine_01, loc=<9.2133 10.439 2.3461>, rot=<-0.6018 -0.436 -0.3926 -0.5419>}
 thigh_l {parent=pelvis, loc=<-1.0855 -0.2109 -8.7053>, rot=<0.0009 0.0355 0.0155 -0.9993>}
 calf_l {parent=thigh_l, loc=<-42.5284 -0.0019 -0.7297>, rot=<0.0006 -0.0161 -0.0391 -0.9991>}
 calf_twist_01_l {parent=calf_l, loc=<-35.8483 -1.9197 -0.5794>, rot=<0.0237 0.0095 -0.0035 -0.9997>}
 calf_twist_02_l {parent=calf_l, loc=<-19.1546 -1.7943 -0.2638>, rot=<0.0237 0.0095 -0.0035 -0.9997>}
 foot_l {parent=calf_l, loc=<-46.7402 -2.5014 -0.592>, rot=<0.0427 -0.0238 0.0353 -0.9982>}
 ball_l {parent=foot_l, loc=<-3.5658 10.8066 -0.1544>, rot=<0.0001 -0.0001 -0.7063 -0.7079>}
 dyn_spur_l_MSTR {parent=foot_l, loc=<0.2042 -8.0539 -0.3512>, rot=<0.7024 -0.1053 0.0072 -0.7039>}
 dyn_spur_l {parent=dyn_spur_l_MSTR, loc=<0 0 0>, rot=<0 0 0 1>}
 deform_knee_in_base_l {parent=calf_l, loc=<0.8031 -5.0231 0.2456>, rot=<0 0 0 -1>}
 deform_knee_in_l {parent=deform_knee_in_base_l, loc=<0 0 0>, rot=<0 0 0 1>}
 deform_knee_out_base_l {parent=calf_l, loc=<0 4 0>, rot=<0 0 0 1>}
 deform_knee_out_l {parent=deform_knee_out_base_l, loc=<0 0 0>, rot=<0 0 0 1>}
 thigh_twist_01_l {parent=thigh_l, loc=<-27.952 0 0>, rot=<0.0013 0 0 -1>}
 deform_groin_base_l {parent=thigh_l, loc=<0.4801 5.6776 -0.2486>, rot=<0 0 0 -1>}
 deform_groin_l {parent=deform_groin_base_l, loc=<0 0 0>, rot=<0 0 0 1>}
 thigh_r {parent=pelvis, loc=<-1.0853 -0.2109 8.7053>, rot=<0.0355 -0.0009 0.9993 0.0155>}
 calf_r {parent=thigh_r, loc=<42.5286 0.0019 0.7297>, rot=<0.0006 -0.0161 -0.0391 -0.9991>}
 calf_twist_01_r {parent=calf_r, loc=<35.8485 1.9197 0.5794>, rot=<0.0237 0.0095 -0.0035 -0.9997>}
 calf_twist_02_r {parent=calf_r, loc=<19.1546 1.7943 0.2638>, rot=<0.0237 0.0095 -0.0035 -0.9997>}
 foot_r {parent=calf_r, loc=<46.7403 2.5014 0.592>, rot=<0.0427 -0.0238 0.0353 -0.9982>}
 ball_r {parent=foot_r, loc=<3.5658 -10.8066 0.1544>, rot=<0.0001 -0.0001 -0.7063 -0.7079>}
 dyn_spur_r_MSTR {parent=foot_r, loc=<-0.2043 8.0539 0.3512>, rot=<-0.0072 -0.7039 0.7024 0.1053>}
 dyn_spur_r {parent=dyn_spur_r_MSTR, loc=<0 0 0>, rot=<0 0 0 1>}
 deform_knee_in_base_r {parent=calf_r, loc=<-0.8031 5.0231 -0.2457>, rot=<0 0 0 -1>}
 deform_knee_in_r {parent=deform_knee_in_base_r, loc=<0 0 0>, rot=<0 0 0 1>}
 deform_knee_out_base_r {parent=calf_r, loc=<0 -4 0>, rot=<0 0 0 1>}
 deform_knee_out_r {parent=deform_knee_out_base_r, loc=<0 0 0>, rot=<0 0 0 1>}
 thigh_twist_01_r {parent=thigh_r, loc=<27.952 0 0>, rot=<0.0013 0 0 -1>}
 deform_groin_base_r {parent=thigh_r, loc=<-0.4799 -5.6776 0.2485>, rot=<0 0 0 -1>}
 deform_groin_r {parent=deform_groin_base_r, loc=<0 0 0>, rot=<0 0 0 1>}
 deform_glute_base_l {parent=pelvis, loc=<-7.3466 -9.0213 -9.1245>, rot=<0.0172 0.031 -0.4477 -0.8935>}
 deform_glute_l {parent=deform_glute_base_l, loc=<0 0 0>, rot=<0 0 0 1>}
 deform_glute_base_r {parent=pelvis, loc=<-7.3464 -9.0213 9.1245>, rot=<0.031 -0.0172 0.8935 -0.4477>}
 deform_glute_r {parent=deform_glute_base_r, loc=<0 0 0>, rot=<0 0 0 1>}
 dyn_skirt_sidebk_r_MSTR {parent=pelvis, loc=<6.8859 -6.237 9.1138>, rot=<0.8318 0.1144 -0.3179 -0.4404>}
 dyn_skirt_sidebk_r_01 {parent=dyn_skirt_sidebk_r_MSTR, loc=<0 0 0>, rot=<0 0 0 1>}
 dyn_skirt_sidebk_r_02 {parent=dyn_skirt_sidebk_r_01, loc=<-7.51 0 0.0002>, rot=<0 0 0 -1>}
 dyn_skirt_side_r_MSTR {parent=pelvis, loc=<5.4924 2.4299 12.7081>, rot=<0.9141 0.0391 -0.4027 0.027>}
 dyn_skirt_side_r_01 {parent=dyn_skirt_side_r_MSTR, loc=<0 0 0>, rot=<0 0 0 1>}
 dyn_skirt_side_r_02 {parent=dyn_skirt_side_r_01, loc=<-7.4701 -0.0001 0.0003>, rot=<-0.0001 0 0.0001 -1>}
 dyn_skirt_sidefr_r_MSTR {parent=pelvis, loc=<4.2491 7.3006 7.2964>, rot=<0.824 -0.0779 -0.3082 0.469>}
 dyn_skirt_sidefr_r_01 {parent=dyn_skirt_sidefr_r_MSTR, loc=<0 0 0>, rot=<0 0 0 1>}
 dyn_skirt_sidefr_r_02 {parent=dyn_skirt_sidefr_r_01, loc=<-6.8079 0.0001 -0.0002>, rot=<0.0001 0 -0.0001 -1>}
 dyn_skirt_front_l_MSTR {parent=pelvis, loc=<4.5285 8.6722 0>, rot=<0.2204 0.6719 0.6719 0.2204>}
 dyn_skirt_front_l_01 {parent=dyn_skirt_front_l_MSTR, loc=<0 0 0>, rot=<0 0 0 1>}
 dyn_skirt_front_l_02 {parent=dyn_skirt_front_l_01, loc=<7.2395 0 0>, rot=<0 0 0 1>}
 dyn_skirt_side_l_MSTR {parent=pelvis, loc=<5.492 2.4299 -12.7081>, rot=<0.0391 -0.9141 -0.027 -0.4027>}
 dyn_skirt_side_l_01 {parent=dyn_skirt_side_l_MSTR, loc=<0 0 0>, rot=<0 0 0 1>}
 dyn_skirt_side_l_02 {parent=dyn_skirt_side_l_01, loc=<7.4698 0 0>, rot=<-0.0001 0 0.0001 -1>}
 dyn_skirt_back_MSTR {parent=pelvis, loc=<6.6572 -8.3974 0>, rot=<0.2295 -0.6753 0.6636 -0.2256>}
 dyn_skirt_back_01 {parent=dyn_skirt_back_MSTR, loc=<0 0 0>, rot=<0 0 0 1>}
 dyn_skirt_back_02 {parent=dyn_skirt_back_01, loc=<7.4887 0 0>, rot=<0 0 0 1>}
 dyn_skirt_sidebk_l_MSTR {parent=pelvis, loc=<6.8856 -6.237 -9.1139>, rot=<0.1144 -0.8318 0.4404 -0.3179>}
 dyn_skirt_sidebk_l_01 {parent=dyn_skirt_sidebk_l_MSTR, loc=<0 0 0>, rot=<0 0 0 1>}
 dyn_skirt_sidebk_l_02 {parent=dyn_skirt_sidebk_l_01, loc=<7.5098 0 0>, rot=<0 0 0 -1>}
 dyn_skirt_sidefr_l_MSTR {parent=pelvis, loc=<4.2494 7.3006 -7.2964>, rot=<-0.0779 -0.824 -0.469 -0.3082>}
 dyn_skirt_sidefr_l_01 {parent=dyn_skirt_sidefr_l_MSTR, loc=<0 0 0>, rot=<0 0 0 1>}
 dyn_skirt_sidefr_l_02 {parent=dyn_skirt_sidefr_l_01, loc=<6.8081 0 0>, rot=<0.0001 0 -0.0001 -1>}
 dyn_skirt_front_MSTR {parent=pelvis, loc=<4.5285 8.6722 0>, rot=<0.2204 0.6719 0.6719 0.2204>}
 dyn_skirt_front_01 {parent=dyn_skirt_front_MSTR, loc=<0 0 0>, rot=<0 0 0 1>}
 dyn_skirt_front_02 {parent=dyn_skirt_front_01, loc=<7.2395 0 0>, rot=<0 0 0 1>}
]
variables [
 CurrentAngle {type=double, default=0}
 EaseFactor {type=double, default=0.01}
 NewAngle {type=double, default=0}
 RotationFactor {type=double, default=10}
 SmoothedSpeed {type=double, default=5}
 SpurSpeed {type=double, default=1}
]
