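# Physics asset for the F_MED_CandyGirl player body, covering its left and right dyn_string bone chains, apparently for secondary motion of dangling strings.
# Each side has capsules on a MSTR bone and three segments, linked in sequence by constraints with narrow swing and tighter twist limits, slightly stiffer at the first joint.
# Large spine_04 and spine_05 capsules cover the upper torso; no constraint connects them to the strings.
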PhysicsAsset /BRCosmetics/Characters/Player/Female/Medium/Bodies/F_MED_CandyGirl/Mesh/F_MED_CandyGirl_Physics
bodies {
 dyn_string_l_01 [{
   shapeType=Capsule
   length=5.8116
   radius=2.1884
   shapeName=dyn_string_l_01_capsule
   center=<4.391 -0.016 1.13>
   rotation=<rot 0 -90.0743 -81.0636>
  }]
 dyn_string_l_02 [{
   shapeType=Capsule
   length=5.4392
   radius=1.6142
   shapeName=dyn_string_l_02_capsule
   center=<2.5658 0.0014 0.4415>
   rotation=<rot 0 89.2529 -89.7387>
  }]
 dyn_string_l_03 [{shapeType=Capsule, length=3.1371, radius=1.2939, shapeName=dyn_string_l_03_capsule, center=<1.9929 0 0>, rotation=<rot 90 0 0>}]
 dyn_string_l_MSTR [{
   shapeType=Capsule
   length=1.3603
   radius=0.9829
   shapeName=dyn_string_l_MSTR_capsule
   center=<0.334 0.1204 -0.1478>
   rotation=<rot 0 100.5103 -64.7395>
  }]
 dyn_string_r_01 [{
   shapeType=Capsule
   length=5.8306
   radius=2.1694
   shapeName=dyn_string_r_01_capsule
   center=<-4.1332 0.0102 -1.1>
   rotation=<rot 0 -90.4363 -82.7687>
  }]
 dyn_string_r_02 [{
   shapeType=Capsule
   length=5.3879
   radius=1.6654
   shapeName=dyn_string_r_02_capsule
   center=<-2.8553 -0.0209 -0.61>
   rotation=<rot 0 -89.6975 -85.8472>
  }]
 dyn_string_r_03 [{
   shapeType=Capsule
   length=4.2282
   radius=1.1253
   shapeName=dyn_string_r_03_capsule
   center=<-1.7748 -0.0285 0.12>
   rotation=<rot 0 -89.0986 -86.9612>
  }]
 dyn_string_r_MSTR [{
   shapeType=Capsule
   length=1.04
   radius=0.9974
   shapeName=dyn_string_r_MSTR_capsule
   center=<0.0523 -0.0611 0.0666>
   rotation=<rot 0 113.7019 -89.0128>
  }]
 spine_04 [{shapeType=Capsule, length=5.3146, radius=12.0197, shapeName=spine_04_capsule, center=<0.4212 1.4455 0>}]
 spine_05 [{shapeType=Capsule, length=8.7384, radius=10.6288, shapeName=spine_05_capsule, center=<0 -0.9167 0>}]
}
constraints [
 {bone1Name=dyn_string_l_01, bone2Name=dyn_string_l_MSTR, swing1LimitDegrees=10, swing2LimitDegrees=10, twistLimitDegrees=4}
 {bone1Name=dyn_string_l_02, bone2Name=dyn_string_l_01, swing1LimitDegrees=15, swing2LimitDegrees=15, twistLimitDegrees=4}
 {bone1Name=dyn_string_l_03, bone2Name=dyn_string_l_02, swing1LimitDegrees=15, swing2LimitDegrees=15, twistLimitDegrees=4}
 {bone1Name=dyn_string_r_01, bone2Name=dyn_string_r_MSTR, swing1LimitDegrees=10, swing2LimitDegrees=10, twistLimitDegrees=4}
 {bone1Name=dyn_string_r_02, bone2Name=dyn_string_r_01, swing1LimitDegrees=15, swing2LimitDegrees=15, twistLimitDegrees=4}
 {bone1Name=dyn_string_r_03, bone2Name=dyn_string_r_02, swing1LimitDegrees=15, swing2LimitDegrees=15, twistLimitDegrees=4}
]
