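# Physics asset for the skirt of the female medium CandyGirl player body cosmetic.
# Eight skirt chains around the front, back, sides and diagonals each link a tiny MSTR root capsule to two larger capsule segments.
# Identical constraints with tight swing and tighter twist limits join each chain vertically, with none between neighboring chains.
# Pelvis and thigh capsules have no constraints, apparently serving as body colliders for the skirt.
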PhysicsAsset /BRCosmetics/Characters/Player/Female/Medium/Bodies/F_MED_CandyGirl/Mesh/F_MED_CandyGirl_Skirt_Physics
bodies {
 dyn_skirt_back_01 [{shapeType=Capsule, length=10, radius=3.7, shapeName=dyn_skirt_back_01_capsule, center=<2 0 0>, rotation=<rot 0 0 90>}]
 dyn_skirt_back_02 [{shapeType=Capsule, length=10, radius=3.7, shapeName=dyn_skirt_back_02_capsule, center=<2 0 0>, rotation=<rot 0 0 90>}]
 dyn_skirt_back_MSTR [{shapeType=Capsule, length=0.505, radius=0.1, shapeName=dyn_skirt_back_MSTR_capsule, rotation=<rot 0 -90 0>}]
 dyn_skirt_front_01 [{shapeType=Capsule, length=10, radius=3.7, shapeName=dyn_skirt_front_01_capsule, center=<2 0 0>, rotation=<rot 0 0 90>}]
 dyn_skirt_front_02 [{shapeType=Capsule, length=10, radius=3.7, shapeName=dyn_skirt_front_02_capsule, center=<2 0 0>, rotation=<rot 0 0 90>}]
 dyn_skirt_front_MSTR [{shapeType=Capsule, length=0.505, radius=0.1, shapeName=dyn_skirt_front_MSTR_capsule, rotation=<rot 0 -90 0>}]
 dyn_skirt_side_l_01 [{shapeType=Capsule, length=10, radius=3.7, shapeName=dyn_skirt_side_l_01_capsule, center=<2 0 0>, rotation=<rot 0 0 90>}]
 dyn_skirt_side_l_02 [{shapeType=Capsule, length=10, radius=3.7, shapeName=dyn_skirt_side_l_02_capsule, center=<2 0 0>, rotation=<rot 0 0 90>}]
 dyn_skirt_side_l_MSTR [{shapeType=Capsule, length=0.505, radius=0.1, shapeName=dyn_skirt_side_l_MSTR_capsule, rotation=<rot 0 -90 0>}]
 dyn_skirt_side_r_01 [{shapeType=Capsule, length=10, radius=3.7, shapeName=dyn_skirt_side_r_01_capsule, center=<-2 0 0>, rotation=<rot 0 0 90>}]
 dyn_skirt_side_r_02 [{shapeType=Capsule, length=10, radius=3.7, shapeName=dyn_skirt_side_r_02_capsule, center=<-2 0 0>, rotation=<rot 0 0 90>}]
 dyn_skirt_side_r_MSTR [{shapeType=Capsule, length=0.505, radius=0.1, shapeName=dyn_skirt_side_r_MSTR_capsule, rotation=<rot 0 -90 0>}]
 dyn_skirt_sidebk_l_01 [{shapeType=Capsule, length=10, radius=3.7, shapeName=dyn_skirt_sidebk_l_01_capsule, center=<2 0 0>, rotation=<rot 0 0 90>}]
 dyn_skirt_sidebk_l_02 [{shapeType=Capsule, length=10, radius=3.7, shapeName=dyn_skirt_sidebk_l_02_capsule, center=<2 0 0>, rotation=<rot 0 0 90>}]
 dyn_skirt_sidebk_l_MSTR [{shapeType=Capsule, length=0.505, radius=0.1, shapeName=dyn_skirt_sidebk_l_MSTR_capsule, rotation=<rot 0 -90 0>}]
 dyn_skirt_sidebk_r_01 [{shapeType=Capsule, length=10, radius=3.7, shapeName=dyn_skirt_sidebk_r_01_capsule, center=<-2 0 0>, rotation=<rot 0 0 90>}]
 dyn_skirt_sidebk_r_02 [{shapeType=Capsule, length=10, radius=3.7, shapeName=dyn_skirt_sidebk_r_02_capsule, center=<-2 0 0>, rotation=<rot 0 0 90>}]
 dyn_skirt_sidebk_r_MSTR [{shapeType=Capsule, length=0.505, radius=0.1, shapeName=dyn_skirt_sidebk_r_MSTR_capsule, rotation=<rot 0 -90 0>}]
 dyn_skirt_sidefr_l_01 [{shapeType=Capsule, length=10, radius=3.7, shapeName=dyn_skirt_sidefr_l_01_capsule, center=<2 0 0>, rotation=<rot 0 0 90>}]
 dyn_skirt_sidefr_l_02 [{shapeType=Capsule, length=10, radius=3.7, shapeName=dyn_skirt_sidefr_l_02_capsule, center=<2 0 0>, rotation=<rot 0 0 90>}]
 dyn_skirt_sidefr_l_MSTR [{shapeType=Capsule, length=0.505, radius=0.1, shapeName=dyn_skirt_sidefr_l_MSTR_capsule, rotation=<rot 0 -90 0>}]
 dyn_skirt_sidefr_r_01 [{shapeType=Capsule, length=10, radius=3.7, shapeName=dyn_skirt_sidefr_r_01_capsule, center=<-2 0 0>, rotation=<rot 0 0 90>}]
 dyn_skirt_sidefr_r_02 [{shapeType=Capsule, length=10, radius=3.7, shapeName=dyn_skirt_sidefr_r_02_capsule, center=<-2 0 0>, rotation=<rot 0 0 90>}]
 dyn_skirt_sidefr_r_MSTR [{shapeType=Capsule, length=0.505, radius=0.1, shapeName=dyn_skirt_sidefr_r_MSTR_capsule, rotation=<rot 0 -90 0>}]
 pelvis [{
   shapeType=Capsule
   length=11.8332
   radius=10.8338
   shapeName=pelvis_capsule
   center=<-4.2727 -1.0911 -0.0048>
   rotation=<rot -0.6549 -90 -0.059>
  }]
 thigh_l [{shapeType=Capsule, length=26.6785, radius=8.2545, shapeName=thigh_l_capsule, center=<-19.3392 1.0269 -0.0277>, rotation=<rot 90 0 0>}]
 thigh_r [{shapeType=Capsule, length=26.6785, radius=8.2545, shapeName=thigh_r_capsule, center=<18.8752 -0.9923 0.0133>, rotation=<rot 90 0 0>}]
}
constraints [
 {bone1Name=dyn_skirt_back_02, bone2Name=dyn_skirt_back_01, swing1LimitDegrees=20, swing2LimitDegrees=20, twistLimitDegrees=4}
 {bone1Name=dyn_skirt_front_02, bone2Name=dyn_skirt_front_01, swing1LimitDegrees=20, swing2LimitDegrees=20, twistLimitDegrees=4}
 {bone1Name=dyn_skirt_side_l_02, bone2Name=dyn_skirt_side_l_01, swing1LimitDegrees=20, swing2LimitDegrees=20, twistLimitDegrees=4}
 {bone1Name=dyn_skirt_side_r_02, bone2Name=dyn_skirt_side_r_01, swing1LimitDegrees=20, swing2LimitDegrees=20, twistLimitDegrees=4}
 {bone1Name=dyn_skirt_sidebk_l_02, bone2Name=dyn_skirt_sidebk_l_01, swing1LimitDegrees=20, swing2LimitDegrees=20, twistLimitDegrees=4}
 {bone1Name=dyn_skirt_sidebk_r_02, bone2Name=dyn_skirt_sidebk_r_01, swing1LimitDegrees=20, swing2LimitDegrees=20, twistLimitDegrees=4}
 {bone1Name=dyn_skirt_sidefr_l_02, bone2Name=dyn_skirt_sidefr_l_01, swing1LimitDegrees=20, swing2LimitDegrees=20, twistLimitDegrees=4}
 {bone1Name=dyn_skirt_sidefr_r_02, bone2Name=dyn_skirt_sidefr_r_01, swing1LimitDegrees=20, swing2LimitDegrees=20, twistLimitDegrees=4}
 {bone1Name=dyn_skirt_back_01, bone2Name=dyn_skirt_back_MSTR, swing1LimitDegrees=20, swing2LimitDegrees=20, twistLimitDegrees=4}
 {bone1Name=dyn_skirt_front_01, bone2Name=dyn_skirt_front_MSTR, swing1LimitDegrees=20, swing2LimitDegrees=20, twistLimitDegrees=4}
 {bone1Name=dyn_skirt_side_l_01, bone2Name=dyn_skirt_side_l_MSTR, swing1LimitDegrees=20, swing2LimitDegrees=20, twistLimitDegrees=4}
 {bone1Name=dyn_skirt_side_r_01, bone2Name=dyn_skirt_side_r_MSTR, swing1LimitDegrees=20, swing2LimitDegrees=20, twistLimitDegrees=4}
 {bone1Name=dyn_skirt_sidebk_l_01, bone2Name=dyn_skirt_sidebk_l_MSTR, swing1LimitDegrees=20, swing2LimitDegrees=20, twistLimitDegrees=4}
 {bone1Name=dyn_skirt_sidebk_r_01, bone2Name=dyn_skirt_sidebk_r_MSTR, swing1LimitDegrees=20, swing2LimitDegrees=20, twistLimitDegrees=4}
 {bone1Name=dyn_skirt_sidefr_l_01, bone2Name=dyn_skirt_sidefr_l_MSTR, swing1LimitDegrees=20, swing2LimitDegrees=20, twistLimitDegrees=4}
 {bone1Name=dyn_skirt_sidefr_r_01, bone2Name=dyn_skirt_sidefr_r_MSTR, swing1LimitDegrees=20, swing2LimitDegrees=20, twistLimitDegrees=4}
]
