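# Physics asset for the left spur (SpurL) on the F_MED_CandyGirl female player body cosmetic, apparently for secondary motion.
# It has capsule bodies on dyn_spur_l and a much smaller one on dyn_spur_l_MSTR, joined by a single constraint.
# The constraint locks twist and swing1, leaving swing2 as the only unlocked angular axis, so the spur rotates like a hinge.
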PhysicsAsset /BRCosmetics/Characters/Player/Female/Medium/Bodies/F_MED_CandyGirl/Mesh/F_MED_CandyGirl_SpurL_Physics
bodies {
 dyn_spur_l [{shapeType=Capsule, length=3.7573, radius=3.7161, shapeName=dyn_spur_l_capsule, center=<1.3582 0 0>, rotation=<rot 0 -90 0>}]
 dyn_spur_l_MSTR [{shapeType=Capsule, length=0.505, radius=0.505, shapeName=dyn_spur_l_MSTR_capsule, rotation=<rot 0 -90 0>}]
}
constraints [{
  bone1Name=dyn_spur_l
  bone2Name=dyn_spur_l_MSTR
  swing1LimitDegrees=45
  swing1Motion=Locked
  swing2LimitDegrees=60
  twistLimitDegrees=45
  twistMotion=Locked
 }]
