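# Post-process-style Anim Blueprint for CandyGirl's cosmetic body, adding secondary motion to skirt, strings and boot spurs.
# Copied mesh pose runs through a Control Rig, then bRigidBodyEnabled selects RigidBody simulations or an AnimDynamics chain for strings and skirt.
# The RigidBody branch also spins spurs from smoothed pawn speed; skydiving lowers body and skirt simulation alpha.
# A separate four-node string AnimDynamics chain is unconnected, and Anim_Dynamics_Alpha is never read.
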AnimBlueprint /BRCosmetics/Characters/Player/Female/Medium/Bodies/F_MED_CandyGirl/Mesh/F_MED_CandyGirl_AnimBp
target_skeleton=&/F_MED_CandyGirl_Skeleton
graphs [
 AnimGraph {nodes [
   n1 = Anim Dynamics - Bone: dyn_string_r_01(ComponentPose=Local To Component(), Alpha=0.25, ExternalForce="0,0,0") {
    angularDampingOverride=0.998
    bGravityOverrideInSimSpace=true
    bOverrideAngularDamping=true
    bUseGravityOverride=true
    lODThreshold=1
    alphaBoolBlend {blendOption=Linear}
    boundBone {boneName=dyn_string_r_01}
    componentAppliedLinearAccClamp=<1000 1000 1000>
    componentLinearAccScale=<0.2 0.2 0.2>
    componentLinearVelScale=<0.2 0.2 0.2>
    gravityOverride {z=-980}
    physicsBodyDefinitions [{
      sphereCollisionRadius=0
      boundBone {boneName=dyn_string_r_01}
      boxExtents=<10 6 6>
      constraintSetup {angularTargetAxis=AxisZ, twistAxis=AxisZ, angularLimitsMax=<25 45 0>, angularLimitsMin=<0 -45 0>, angularTarget=<0 0 0>}
      localJointOffset=<0 0 -1>
     }]
   }
   n2 = Anim Dynamics - Bone: dyn_string_l_01(ComponentPose=@n1, Alpha=0.25, ExternalForce="0,0,0") {
    angularDampingOverride=0.998
    bGravityOverrideInSimSpace=true
    bOverrideAngularDamping=true
    bUseGravityOverride=true
    lODThreshold=1
    alphaBoolBlend {blendOption=Linear}
    boundBone {boneName=dyn_string_l_01}
    componentAppliedLinearAccClamp=<1000 1000 1000>
    componentLinearAccScale=<0.2 0.2 0.2>
    componentLinearVelScale=<0.2 0.2 0.2>
    gravityOverride {z=-980}
    physicsBodyDefinitions [{
      sphereCollisionRadius=0
      boundBone {boneName=dyn_string_l_01}
      boxExtents=<10 6 6>
      constraintSetup {angularTargetAxis=AxisZ, twistAxis=AxisZ, angularLimitsMax=<25 45 0>, angularLimitsMin=<0 -45 0>, angularTarget=<0 0 0>}
      localJointOffset=<0 0 -1>
     }]
   }
   n3 = Anim Dynamics - Bone: dyn_string_r_02(ComponentPose=@n2, Alpha=0.3, ExternalForce="0,0,0") {
    angularDampingOverride=0.998
    bGravityOverrideInSimSpace=true
    bOverrideAngularDamping=true
    bUseGravityOverride=true
    lODThreshold=1
    alphaBoolBlend {blendOption=Linear}
    boundBone {boneName=dyn_string_r_02}
    componentAppliedLinearAccClamp=<1000 1000 1000>
    componentLinearAccScale=<0.2 0.2 0.2>
    componentLinearVelScale=<0.2 0.2 0.2>
    gravityOverride {z=-980}
    physicsBodyDefinitions [{
      sphereCollisionRadius=0
      boundBone {boneName=dyn_string_r_02}
      boxExtents=<10 6 6>
      constraintSetup {angularTargetAxis=AxisZ, twistAxis=AxisZ, angularLimitsMax=<25 45 0>, angularLimitsMin=<0 -45 0>, angularTarget=<0 0 0>}
      localJointOffset=<0 0 -1>
     }]
   }
   n4 = Anim Dynamics - Bone: dyn_string_l_02(ComponentPose=@n3, Alpha=0.3, ExternalForce="0,0,0") {
    angularDampingOverride=0.998
    bGravityOverrideInSimSpace=true
    bOverrideAngularDamping=true
    bUseGravityOverride=true
    lODThreshold=1
    alphaBoolBlend {blendOption=Linear}
    boundBone {boneName=dyn_string_l_02}
    componentAppliedLinearAccClamp=<1000 1000 1000>
    componentLinearAccScale=<0.2 0.2 0.2>
    componentLinearVelScale=<0.2 0.2 0.2>
    gravityOverride {z=-980}
    physicsBodyDefinitions [{
      sphereCollisionRadius=0
      boundBone {boneName=dyn_string_l_02}
      boxExtents=<10 6 6>
      constraintSetup {angularTargetAxis=AxisZ, twistAxis=AxisZ, angularLimitsMax=<25 45 0>, angularLimitsMin=<0 -45 0>, angularTarget=<0 0 0>}
      localJointOffset=<0 0 -1>
     }]
   }
   Component To Local(ComponentPose=@n4)
   n5 = Anim Dynamics - Bone: dyn_string_l_01(ComponentPose=Local To Component(LocalPose=Use cached pose 'CR_Cache'()), Alpha=1, ExternalForce="0,0,0") {
    angularDampingOverride=0.9
    angularSpringConstant=500
    bAngularSpring=true
    bOverrideAngularDamping=true
    bUseGravityOverride=true
    simulationSpace=RootRelative
    alphaBoolBlend {blendOption=Linear}
    boundBone {boneName=dyn_string_l_01}
    componentLinearAccScale=<0.4 0.4 0.4>
    componentLinearVelScale=<0.4 0.4 0.4>
    gravityOverride {z=-980}
    physicsBodyDefinitions [{
      boundBone {boneName=dyn_string_l_01}
      boxExtents=<15 15 15>
      constraintSetup {angularLimitsMax=<0 0 20>, angularLimitsMin=<0 -40 -20>, angularTarget=<1 0 0>}
      localJointOffset=<1.5 0 0>
     }]
    relativeSpaceBone {boneName=pelvis}
   }
   n6 = Anim Dynamics - Bone: dyn_string_l_02(ComponentPose=@n5, Alpha=1, ExternalForce="0,0,0") {
    angularDampingOverride=0.9
    angularSpringConstant=500
    bAngularSpring=true
    bOverrideAngularDamping=true
    bUseGravityOverride=true
    simulationSpace=RootRelative
    alphaBoolBlend {blendOption=Linear}
    boundBone {boneName=dyn_string_l_02}
    componentLinearAccScale=<0.4 0.4 0.4>
    componentLinearVelScale=<0.4 0.4 0.4>
    gravityOverride {z=-980}
    physicsBodyDefinitions [{
      boundBone {boneName=dyn_string_l_02}
      boxExtents=<15 15 15>
      constraintSetup {angularLimitsMax=<0 5 20>, angularLimitsMin=<0 -40 -20>, angularTarget=<1 0 0>}
      localJointOffset=<1.5 0 0>
     }]
    relativeSpaceBone {boneName=pelvis}
   }
   n7 = Anim Dynamics - Bone: dyn_string_l_03(ComponentPose=@n6, Alpha=1, ExternalForce="0,0,0") {
    angularDampingOverride=0.9
    angularSpringConstant=500
    bAngularSpring=true
    bOverrideAngularDamping=true
    bUseGravityOverride=true
    simulationSpace=RootRelative
    alphaBoolBlend {blendOption=Linear}
    boundBone {boneName=dyn_string_l_03}
    componentLinearAccScale=<0.4 0.4 0.4>
    componentLinearVelScale=<0.4 0.4 0.4>
    gravityOverride {z=-980}
    physicsBodyDefinitions [{
      boundBone {boneName=dyn_string_l_03}
      boxExtents=<15 15 15>
      constraintSetup {angularLimitsMax=<0 0 20>, angularLimitsMin=<0 -40 -20>, angularTarget=<1 0 0>}
      localJointOffset=<1.5 0 0>
     }]
    relativeSpaceBone {boneName=pelvis}
   }
   n8 = Anim Dynamics - Bone: dyn_string_r_01(ComponentPose=@n7, Alpha=1, ExternalForce="0,0,0") {
    angularDampingOverride=0.9
    angularSpringConstant=500
    bAngularSpring=true
    bOverrideAngularDamping=true
    bUseGravityOverride=true
    simulationSpace=RootRelative
    alphaBoolBlend {blendOption=Linear}
    boundBone {boneName=dyn_string_r_01}
    componentLinearAccScale=<0.4 0.4 0.4>
    componentLinearVelScale=<0.4 0.4 0.4>
    gravityOverride {z=-980}
    physicsBodyDefinitions [{
      boundBone {boneName=dyn_string_r_01}
      boxExtents=<15 15 15>
      constraintSetup {angularLimitsMax=<0 0 20>, angularLimitsMin=<0 -40 -20>, angularTarget=<1 0 0>}
      localJointOffset=<-1.5 0 0>
     }]
    relativeSpaceBone {boneName=pelvis}
   }
   n9 = Anim Dynamics - Bone: dyn_string_r_02(ComponentPose=@n8, Alpha=1, ExternalForce="0,0,0") {
    angularDampingOverride=0.9
    angularSpringConstant=500
    bAngularSpring=true
    bOverrideAngularDamping=true
    bUseGravityOverride=true
    simulationSpace=RootRelative
    alphaBoolBlend {blendOption=Linear}
    boundBone {boneName=dyn_string_r_02}
    componentLinearAccScale=<0.4 0.4 0.4>
    componentLinearVelScale=<0.4 0.4 0.4>
    gravityOverride {z=-980}
    physicsBodyDefinitions [{
      boundBone {boneName=dyn_string_r_02}
      boxExtents=<15 15 15>
      constraintSetup {angularLimitsMax=<0 5 20>, angularLimitsMin=<0 -40 -20>, angularTarget=<1 0 0>}
      localJointOffset=<-1.5 0 0>
     }]
    relativeSpaceBone {boneName=pelvis}
   }
   n10 = Anim Dynamics - Bone: dyn_string_r_03(ComponentPose=@n9, Alpha=1, ExternalForce="0,0,0") {
    angularDampingOverride=0.9
    angularSpringConstant=500
    bAngularSpring=true
    bOverrideAngularDamping=true
    bUseGravityOverride=true
    simulationSpace=RootRelative
    alphaBoolBlend {blendOption=Linear}
    boundBone {boneName=dyn_string_r_03}
    componentLinearAccScale=<0.4 0.4 0.4>
    componentLinearVelScale=<0.4 0.4 0.4>
    gravityOverride {z=-980}
    physicsBodyDefinitions [{
      boundBone {boneName=dyn_string_r_03}
      boxExtents=<15 15 15>
      constraintSetup {angularLimitsMax=<0 0 20>, angularLimitsMin=<0 -40 -20>, angularTarget=<1 0 0>}
      localJointOffset=<-1.5 0 0>
     }]
    relativeSpaceBone {boneName=pelvis}
   }
   n11 = Anim Dynamics - Bone: dyn_skirt_front_01(ComponentPose=@n10, Alpha=1, ExternalForce="0,0,0") {
    angularDampingOverride=0.9
    angularSpringConstant=500
    bAngularSpring=true
    bOverrideAngularDamping=true
    bUseGravityOverride=true
    simulationSpace=RootRelative
    alphaBoolBlend {blendOption=Linear}
    boundBone {boneName=dyn_skirt_front_01}
    componentLinearAccScale=<0.4 0.4 0.4>
    componentLinearVelScale=<0.4 0.4 0.4>
    gravityOverride {z=-980}
    physicsBodyDefinitions [{
      boundBone {boneName=dyn_skirt_front_01}
      boxExtents=<15 15 15>
      constraintSetup {angularLimitsMax=<0 0 5>, angularLimitsMin=<0 -40 -5>, angularTarget=<1 0 0>}
      localJointOffset=<2 0 0>
     }]
    relativeSpaceBone {boneName=pelvis}
   }
   n12 = Anim Dynamics - Bone: dyn_skirt_front_02(ComponentPose=@n11, Alpha=1, ExternalForce="0,0,0") {
    angularDampingOverride=0.9
    angularSpringConstant=500
    bAngularSpring=true
    bOverrideAngularDamping=true
    bUseGravityOverride=true
    simulationSpace=RootRelative
    alphaBoolBlend {blendOption=Linear}
    boundBone {boneName=dyn_skirt_front_02}
    componentLinearAccScale=<0.4 0.4 0.4>
    componentLinearVelScale=<0.4 0.4 0.4>
    gravityOverride {z=-980}
    physicsBodyDefinitions [{
      boundBone {boneName=dyn_skirt_front_02}
      boxExtents=<15 15 15>
      constraintSetup {angularLimitsMax=<0 0 5>, angularLimitsMin=<0 -40 -5>, angularTarget=<1 0 0>}
      localJointOffset=<2 0 0>
     }]
    relativeSpaceBone {boneName=pelvis}
   }
   n13 = Anim Dynamics - Bone: dyn_skirt_sidefr_l_01(ComponentPose=@n12, Alpha=1, ExternalForce="0,0,0") {
    angularDampingOverride=0.9
    angularSpringConstant=500
    bAngularSpring=true
    bOverrideAngularDamping=true
    bUseGravityOverride=true
    simulationSpace=RootRelative
    alphaBoolBlend {blendOption=Linear}
    boundBone {boneName=dyn_skirt_sidefr_l_01}
    componentLinearAccScale=<0.4 0.4 0.4>
    componentLinearVelScale=<0.4 0.4 0.4>
    gravityOverride {z=-980}
    physicsBodyDefinitions [{
      boundBone {boneName=dyn_skirt_sidefr_l_01}
      boxExtents=<15 15 15>
      constraintSetup {angularLimitsMax=<0 0 5>, angularLimitsMin=<0 -40 -5>, angularTarget=<1 0 0>}
      localJointOffset=<2 0 0>
     }]
    relativeSpaceBone {boneName=pelvis}
   }
   n14 = Anim Dynamics - Bone: dyn_skirt_sidefr_l_02(ComponentPose=@n13, Alpha=1, ExternalForce="0,0,0") {
    angularDampingOverride=0.9
    angularSpringConstant=500
    bAngularSpring=true
    bOverrideAngularDamping=true
    bUseGravityOverride=true
    simulationSpace=RootRelative
    alphaBoolBlend {blendOption=Linear}
    boundBone {boneName=dyn_skirt_sidefr_l_02}
    componentLinearAccScale=<0.4 0.4 0.4>
    componentLinearVelScale=<0.4 0.4 0.4>
    gravityOverride {z=-980}
    physicsBodyDefinitions [{
      boundBone {boneName=dyn_skirt_sidefr_l_02}
      boxExtents=<15 15 15>
      constraintSetup {angularLimitsMax=<0 0 5>, angularLimitsMin=<0 -40 -5>, angularTarget=<1 0 0>}
      localJointOffset=<2 0 0>
     }]
    relativeSpaceBone {boneName=pelvis}
   }
   n15 = Anim Dynamics - Bone: dyn_skirt_side_l_01(ComponentPose=@n14, Alpha=1, ExternalForce="0,0,0") {
    angularDampingOverride=0.9
    angularSpringConstant=500
    bAngularSpring=true
    bOverrideAngularDamping=true
    bUseGravityOverride=true
    simulationSpace=RootRelative
    alphaBoolBlend {blendOption=Linear}
    boundBone {boneName=dyn_skirt_side_l_01}
    componentLinearAccScale=<0.4 0.4 0.4>
    componentLinearVelScale=<0.4 0.4 0.4>
    gravityOverride {z=-980}
    physicsBodyDefinitions [{
      boundBone {boneName=dyn_skirt_side_l_01}
      boxExtents=<15 15 15>
      constraintSetup {angularLimitsMax=<0 0 5>, angularLimitsMin=<0 -40 -5>, angularTarget=<1 0 0>}
      localJointOffset=<2 0 0>
     }]
    relativeSpaceBone {boneName=pelvis}
   }
   n16 = Anim Dynamics - Bone: dyn_skirt_side_l_02(ComponentPose=@n15, Alpha=1, ExternalForce="0,0,0") {
    angularDampingOverride=0.9
    angularSpringConstant=500
    bAngularSpring=true
    bOverrideAngularDamping=true
    bUseGravityOverride=true
    simulationSpace=RootRelative
    alphaBoolBlend {blendOption=Linear}
    boundBone {boneName=dyn_skirt_side_l_02}
    componentLinearAccScale=<0.4 0.4 0.4>
    componentLinearVelScale=<0.4 0.4 0.4>
    gravityOverride {z=-980}
    physicsBodyDefinitions [{
      boundBone {boneName=dyn_skirt_side_l_02}
      boxExtents=<15 15 15>
      constraintSetup {angularLimitsMax=<0 0 5>, angularLimitsMin=<0 -40 -5>, angularTarget=<1 0 0>}
      localJointOffset=<2 0 0>
     }]
    relativeSpaceBone {boneName=pelvis}
   }
   n17 = Anim Dynamics - Bone: dyn_skirt_sidebk_l_01(ComponentPose=@n16, Alpha=1, ExternalForce="0,0,0") {
    angularDampingOverride=0.9
    angularSpringConstant=500
    bAngularSpring=true
    bOverrideAngularDamping=true
    bUseGravityOverride=true
    simulationSpace=RootRelative
    alphaBoolBlend {blendOption=Linear}
    boundBone {boneName=dyn_skirt_sidebk_l_01}
    componentLinearAccScale=<0.4 0.4 0.4>
    componentLinearVelScale=<0.4 0.4 0.4>
    gravityOverride {z=-980}
    physicsBodyDefinitions [{
      boundBone {boneName=dyn_skirt_sidebk_l_01}
      boxExtents=<15 15 15>
      constraintSetup {angularLimitsMax=<0 0 5>, angularLimitsMin=<0 -40 -5>, angularTarget=<1 0 0>}
      localJointOffset=<2 0 0>
     }]
    relativeSpaceBone {boneName=pelvis}
   }
   n18 = Anim Dynamics - Bone: dyn_skirt_sidebk_l_02(ComponentPose=@n17, Alpha=1, ExternalForce="0,0,0") {
    angularDampingOverride=0.9
    angularSpringConstant=500
    bAngularSpring=true
    bOverrideAngularDamping=true
    bUseGravityOverride=true
    simulationSpace=RootRelative
    alphaBoolBlend {blendOption=Linear}
    boundBone {boneName=dyn_skirt_sidebk_l_02}
    componentLinearAccScale=<0.4 0.4 0.4>
    componentLinearVelScale=<0.4 0.4 0.4>
    gravityOverride {z=-980}
    physicsBodyDefinitions [{
      boundBone {boneName=dyn_skirt_sidebk_l_02}
      boxExtents=<15 15 15>
      constraintSetup {angularLimitsMax=<0 0 5>, angularLimitsMin=<0 -40 -5>, angularTarget=<1 0 0>}
      localJointOffset=<2 0 0>
     }]
    relativeSpaceBone {boneName=pelvis}
   }
   n19 = Anim Dynamics - Bone: dyn_skirt_back_01(ComponentPose=@n18, Alpha=1, ExternalForce="0,0,0") {
    angularDampingOverride=0.9
    angularSpringConstant=500
    bAngularSpring=true
    bOverrideAngularDamping=true
    bUseGravityOverride=true
    simulationSpace=RootRelative
    alphaBoolBlend {blendOption=Linear}
    boundBone {boneName=dyn_skirt_back_01}
    componentLinearAccScale=<0.4 0.4 0.4>
    componentLinearVelScale=<0.4 0.4 0.4>
    gravityOverride {z=-980}
    physicsBodyDefinitions [{
      boundBone {boneName=dyn_skirt_back_01}
      boxExtents=<15 15 15>
      constraintSetup {angularLimitsMax=<0 0 5>, angularLimitsMin=<0 -40 -5>, angularTarget=<1 0 0>}
      localJointOffset=<2 0 0>
     }]
    relativeSpaceBone {boneName=pelvis}
   }
   n20 = Anim Dynamics - Bone: dyn_skirt_back_02(ComponentPose=@n19, Alpha=1, ExternalForce="0,0,0") {
    angularDampingOverride=0.9
    angularSpringConstant=500
    bAngularSpring=true
    bOverrideAngularDamping=true
    bUseGravityOverride=true
    simulationSpace=RootRelative
    alphaBoolBlend {blendOption=Linear}
    boundBone {boneName=dyn_skirt_back_02}
    componentLinearAccScale=<0.4 0.4 0.4>
    componentLinearVelScale=<0.4 0.4 0.4>
    gravityOverride {z=-980}
    physicsBodyDefinitions [{
      boundBone {boneName=dyn_skirt_back_02}
      boxExtents=<15 15 15>
      constraintSetup {angularLimitsMax=<0 0 5>, angularLimitsMin=<0 -40 -5>, angularTarget=<1 0 0>}
      localJointOffset=<2 0 0>
     }]
    relativeSpaceBone {boneName=pelvis}
   }
   n21 = Anim Dynamics - Bone: dyn_skirt_sidefr_r_01(ComponentPose=@n20, Alpha=1, ExternalForce="0,0,0") {
    angularDampingOverride=0.9
    angularSpringConstant=500
    bAngularSpring=true
    bOverrideAngularDamping=true
    bUseGravityOverride=true
    simulationSpace=RootRelative
    alphaBoolBlend {blendOption=Linear}
    boundBone {boneName=dyn_skirt_sidefr_r_01}
    componentLinearAccScale=<0.4 0.4 0.4>
    componentLinearVelScale=<0.4 0.4 0.4>
    gravityOverride {z=-980}
    physicsBodyDefinitions [{
      boundBone {boneName=dyn_skirt_sidefr_r_01}
      boxExtents=<15 15 15>
      constraintSetup {angularLimitsMax=<0 0 5>, angularLimitsMin=<0 -40 -5>, angularTarget=<1 0 0>}
      localJointOffset=<-2 0 0>
     }]
    relativeSpaceBone {boneName=pelvis}
   }
   n22 = Anim Dynamics - Bone: dyn_skirt_sidefr_r_02(ComponentPose=@n21, Alpha=1, ExternalForce="0,0,0") {
    angularDampingOverride=0.9
    angularSpringConstant=500
    bAngularSpring=true
    bOverrideAngularDamping=true
    bUseGravityOverride=true
    simulationSpace=RootRelative
    alphaBoolBlend {blendOption=Linear}
    boundBone {boneName=dyn_skirt_sidefr_r_02}
    componentLinearAccScale=<0.4 0.4 0.4>
    componentLinearVelScale=<0.4 0.4 0.4>
    gravityOverride {z=-980}
    physicsBodyDefinitions [{
      boundBone {boneName=dyn_skirt_sidefr_r_02}
      boxExtents=<15 15 15>
      constraintSetup {angularLimitsMax=<0 0 5>, angularLimitsMin=<0 -40 -5>, angularTarget=<1 0 0>}
      localJointOffset=<-2 0 0>
     }]
    relativeSpaceBone {boneName=pelvis}
   }
   n23 = Anim Dynamics - Bone: dyn_skirt_side_r_01(ComponentPose=@n22, Alpha=1, ExternalForce="0,0,0") {
    angularDampingOverride=0.9
    angularSpringConstant=500
    bAngularSpring=true
    bOverrideAngularDamping=true
    bUseGravityOverride=true
    simulationSpace=RootRelative
    alphaBoolBlend {blendOption=Linear}
    boundBone {boneName=dyn_skirt_side_r_01}
    componentLinearAccScale=<0.4 0.4 0.4>
    componentLinearVelScale=<0.4 0.4 0.4>
    gravityOverride {z=-980}
    physicsBodyDefinitions [{
      boundBone {boneName=dyn_skirt_side_r_01}
      boxExtents=<15 15 15>
      constraintSetup {angularLimitsMax=<0 0 5>, angularLimitsMin=<0 -40 -5>, angularTarget=<1 0 0>}
      localJointOffset=<-2 0 0>
     }]
    relativeSpaceBone {boneName=pelvis}
   }
   n24 = Anim Dynamics - Bone: dyn_skirt_side_r_02(ComponentPose=@n23, Alpha=1, ExternalForce="0,0,0") {
    angularDampingOverride=0.9
    angularSpringConstant=500
    bAngularSpring=true
    bOverrideAngularDamping=true
    bUseGravityOverride=true
    simulationSpace=RootRelative
    alphaBoolBlend {blendOption=Linear}
    boundBone {boneName=dyn_skirt_side_r_02}
    componentLinearAccScale=<0.4 0.4 0.4>
    componentLinearVelScale=<0.4 0.4 0.4>
    gravityOverride {z=-980}
    physicsBodyDefinitions [{
      boundBone {boneName=dyn_skirt_side_r_02}
      boxExtents=<15 15 15>
      constraintSetup {angularLimitsMax=<0 0 5>, angularLimitsMin=<0 -40 -5>, angularTarget=<1 0 0>}
      localJointOffset=<-2 0 0>
     }]
    relativeSpaceBone {boneName=pelvis}
   }
   n25 = Anim Dynamics - Bone: dyn_skirt_sidebk_r_01(ComponentPose=@n24, Alpha=1, ExternalForce="0,0,0") {
    angularDampingOverride=0.9
    angularSpringConstant=500
    bAngularSpring=true
    bOverrideAngularDamping=true
    bUseGravityOverride=true
    simulationSpace=RootRelative
    alphaBoolBlend {blendOption=Linear}
    boundBone {boneName=dyn_skirt_sidebk_r_01}
    componentLinearAccScale=<0.4 0.4 0.4>
    componentLinearVelScale=<0.4 0.4 0.4>
    gravityOverride {z=-980}
    physicsBodyDefinitions [{
      boundBone {boneName=dyn_skirt_sidebk_r_01}
      boxExtents=<15 15 15>
      constraintSetup {angularLimitsMax=<0 0 5>, angularLimitsMin=<0 -40 -5>, angularTarget=<1 0 0>}
      localJointOffset=<-2 0 0>
     }]
    relativeSpaceBone {boneName=pelvis}
   }
   n26 = Anim Dynamics - Bone: dyn_skirt_sidebk_r_02(ComponentPose=@n25, Alpha=1, ExternalForce="0,0,0") {
    angularDampingOverride=0.9
    angularSpringConstant=500
    bAngularSpring=true
    bOverrideAngularDamping=true
    bUseGravityOverride=true
    simulationSpace=RootRelative
    alphaBoolBlend {blendOption=Linear}
    boundBone {boneName=dyn_skirt_sidebk_r_02}
    componentLinearAccScale=<0.4 0.4 0.4>
    componentLinearVelScale=<0.4 0.4 0.4>
    gravityOverride {z=-980}
    physicsBodyDefinitions [{
      boundBone {boneName=dyn_skirt_sidebk_r_02}
      boxExtents=<15 15 15>
      constraintSetup {angularLimitsMax=<0 40 5>, angularLimitsMin=<0 0 -5>, angularTarget=<1 0 0>}
      localJointOffset=<-2 0 0>
     }]
    relativeSpaceBone {boneName=pelvis}
   }
   n27 = Get RBAN_Alpha() -> (RBAN_Alpha:double)
   n28 = RigidBody(ComponentPose=Local To Component(LocalPose=Use cached pose 'CR_Cache'()), Alpha=@n27) {
    bOverrideWorldGravity=true
    overridePhysicsAsset=&/F_MED_CandyGirl_Physics
    simulationSpace=BaseBoneSpace
    alphaBoolBlend {blendOption=Linear}
    baseBoneRef {boneName=spine_05}
    overrideWorldGravity {z=-980}
    simSpaceSettings {
     dampingAlpha=0.2
     maxAngularAcceleration=20
     maxAngularVelocity=4
     maxLinearAcceleration=2000
     maxLinearVelocity=400
     velocityScaleZ=0.7
     worldAlpha=1
     externalLinearDragV=<0.2 0.2 0.2>
    }
   }
   n29 = RigidBody(ComponentPose=@n28, Alpha=@n27) {
    bOverrideWorldGravity=true
    overridePhysicsAsset=&/F_MED_CandyGirl_Skirt_Physics
    simulationSpace=BaseBoneSpace
    alphaBoolBlend {blendOption=Linear}
    baseBoneRef {boneName=pelvis}
    overrideWorldGravity {z=-980}
    simSpaceSettings {
     dampingAlpha=0.9
     maxAngularAcceleration=15
     maxAngularVelocity=6
     maxLinearAcceleration=3000
     maxLinearVelocity=300
     velocityScaleZ=0.4
     worldAlpha=1
     externalLinearDragV=<1 1 1>
    }
   }
   n30 = RigidBody(ComponentPose=@n29) {
    bOverrideWorldGravity=true
    overridePhysicsAsset=&/F_MED_CandyGirl_SpurL_Physics
    simulationSpace=BaseBoneSpace
    alphaBoolBlend {blendOption=Linear}
    baseBoneRef {boneName=foot_l}
    overrideWorldGravity {z=-980}
    simSpaceSettings {
     dampingAlpha=0
     maxAngularAcceleration=40
     maxAngularVelocity=12
     maxLinearAcceleration=3500
     maxLinearVelocity=700
     worldAlpha=1
     externalLinearDragV=<0.2 0.2 0.2>
    }
   }
   n31 = RigidBody(ComponentPose=@n30) {
    bOverrideWorldGravity=true
    overridePhysicsAsset=&/F_MED_CandyGirl_SpurR_Physics
    simulationSpace=BaseBoneSpace
    alphaBoolBlend {blendOption=Linear}
    baseBoneRef {boneName=foot_r}
    overrideWorldGravity {z=-980}
    simSpaceSettings {
     dampingAlpha=0
     maxAngularAcceleration=40
     maxAngularVelocity=12
     maxLinearAcceleration=3500
     maxLinearVelocity=700
     worldAlpha=1
     externalLinearDragV=<0.2 0.2 0.2>
    }
   }
   n32 = Get SpinAngle() -> (SpinAngle:double)
   n33 = "Transform (Modify) Bone - Bone: dyn_spur_l_MSTR"!(ComponentPose=@n31, Alpha=1, Translation="0,0,0", Rotation_Roll:float=0, Rotation_Pitch:float=@n32, Rotation_Yaw:float=0, Scale="1,1,1") -> (Pose) {rotationMode=BMM_Additive, rotationSpace=BCS_BoneSpace, alphaBoolBlend {blendOption=Linear}, boneToModify {boneName=dyn_spur_l_MSTR}}
   n34 = "Transform (Modify) Bone - Bone: dyn_spur_r_MSTR"!(ComponentPose=@n33, Alpha=1, Translation="0,0,0", Rotation_Roll:float=0, Rotation_Pitch:float=@n32, Rotation_Yaw:float=0, Scale="1,1,1") -> (Pose) {rotationMode=BMM_Additive, rotationSpace=BCS_BoneSpace, alphaBoolBlend {blendOption=Linear}, boneToModify {boneName=dyn_spur_r_MSTR}}
   Output Pose(Result=Blend Poses by bool(BlendPose_0=Component To Local(ComponentPose=@n34), BlendPose_1=Component To Local(ComponentPose=@n26), bActiveValue=Get bRigidBodyEnabled()) {blendType=Linear}) {layerGroup=None}
   n35 = Control Rig!(
    Source=Copy Pose From Mesh(SourceMeshComponent=Get MeshToCopy() -> (MeshToCopy:Skeletal Mesh Component Object Reference))
    Alpha
    SpurSpeed:double=Get SpurSpeed() -> (SpurSpeed:double)
   ) -> (Pose) {
    alphaBoolBlend {blendOption=Linear}
    controlRigAssetReference {blueprintRigClass=&/F_MED_CandyGirl_CtrlRig.F_MED_CandyGirl_CtrlRig_C}
   }
   Save cached pose 'CR_Cache'(Pose=@n35) {cacheName=CR_Cache}
  ]}
 EventGraph {nodes [
   n1 = Event BlueprintUpdateAnimation()
   > n2 = Sequence!(execute) -> (then_0, then_1, then_2:Exec)
   > n3 = Branch(Condition=Get bIsSkyDiving())
   > Set RBAN_Alpha(RBAN_Alpha:double=0.75) -> (Output_Get:double)
   > Set Anim_Dynamics_Alpha(Anim_Dynamics_Alpha:double=0.75) -> (Output_Get:double)
   Set RBAN_Alpha(execute=@n3.else, RBAN_Alpha=1)
   > Set Anim_Dynamics_Alpha(Anim_Dynamics_Alpha=1)
   n4 = Get isAccelerating() -> (isAccelerating:bool)
   Set SmoothedSpeed(
    execute=@n2.then_2
    SmoothedSpeed:double=FInterpTo(
     Current=Get SmoothedSpeed() -> (SmoothedSpeed:double)
     Target=Get SpurSpeed()
     DeltaTime=@n1.DeltaTimeX
     InterpSpeed=SelectFloat(A=2, B=1, bPickA=@n4)
    )
   ) -> (Output_Get:double)
   n5 = Get RotationFactor() -> (RotationFactor:double)
   n6 = Get SmoothedSpeed()
   > Set SpinAngle(SpinAngle:double=float + float(A=Get SpinAngle(), B=float * float!(A=@n6, B=@n5, C:float=@n1.DeltaTimeX) -> (ReturnValue) {is=title K2Node_PromotableOperator})) -> (Output_Get:double)
   Set SpurSpeed(execute=@n2.then_1, SpurSpeed:double=MapRangeClamped(Value=Get PawnSpeed2D(), InRangeA=0, InRangeB=550, OutRangeB=750)) -> (Output_Get:double)
   > n7 = Branch(Condition="float > float"(A=Get SpurSpeed(), B=Get SmoothedSpeed()))
   > Set isAccelerating(isAccelerating:bool=true) -> (Output_Get:bool)
   Set isAccelerating(execute=@n7.else, isAccelerating=false)
   Event BlueprintInitializeAnimation()
   > Set MeshToCopy(MeshToCopy:Skeletal Mesh Component Object Reference=GetOwnerSkeletalMeshComponent()) -> (Output_Get:Skeletal Mesh Component Object Reference)
  ]}
]
variables [
 Anim_Dynamics_Alpha {type=number}
 isAccelerating {type=boolean}
 MeshToCopy {type=SkeletalMeshComponent}
 RBAN_Alpha {type=number}
 RotationFactor {type=number, default=1}
 SmoothedSpeed {type=number, default=5}
 SpinAngle {type=number}
 SpurSpeed {type=number, default=50}
]
